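# Niagara idle cosmetic effect for the RebelClaw Aviator body, attaching flames and rising embers to the character's feet.
# NE_FootFlame keeps two mesh particles on skeletal bones or sockets, mirrored by partition, feeding scroll offset and smoothed inverse owner velocity into dynamic material parameters.
# NE_FootEmbers spawns short-lived sprites there, moved by curl noise, drag and inherited motion.
# System.SpeedFactor drives ember rate, flame scrolling and a material parameter.
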NiagaraSystem /BRCosmetics/Characters/Player/Female/Medium/Bodies/F_MED_RebelClaw_Aviator/FX/NS_RebelClawAviator_Idle_Feet
dependencies {
 usedDataInterfaces [&NiagaraDataInterfaceCurve, &NiagaraDataInterfaceSkeletalMesh]
 usedDynamicInputs [
  &SetBoolByFloatComparison
  &VectorLength
  &ClampFloat
  &Divide_Float
  &Subtract_Float
  &MakeCustomFloatFromBool
  &InvertBool
 ]
 usedModules [
  &self:SystemSpawnScript.SystemScriptSource.SystemScriptGraph.NiagaraNodeAssignment_0.SetVariables_#1
  &self:SystemSpawnScript.SystemScriptSource.SystemScriptGraph.NiagaraNodeAssignment_5.SetVariables_#2
  &self:SystemSpawnScript.SystemScriptSource.SystemScriptGraph.NiagaraNodeAssignment_6.SetVariables_#3
  &SystemState
  &self:SystemSpawnScript.SystemScriptSource.SystemScriptGraph.NiagaraNodeAssignment_4.SetVariables_#4
  &self:SystemSpawnScript.SystemScriptSource.SystemScriptGraph.NiagaraNodeAssignment_1.SetVariables_#5
  &self:SystemSpawnScript.SystemScriptSource.SystemScriptGraph.NiagaraNodeAssignment_2.SetVariables_#6
  &self:SystemSpawnScript.SystemScriptSource.SystemScriptGraph.NiagaraNodeAssignment_3.SetVariables_#7
  &self:SystemSpawnScript.SystemScriptSource.SystemScriptGraph.NiagaraNodeAssignment_7.SetVariables_#8
 ]
}
emitters {
 NE_FootEmbers {
  emitterUpdateScript [
   &EmitterState()
   &SpawnRate(SpawnRate=&Lerp_Float(A=5, B=12, Alpha=$System.SpeedFactor), Use Spawn Probability=false)
  ]
  particleSpawnScript [
   &PartitionParticles()
   &InitializeParticle(
    Color=&LerpLinearColors(StartColor=<rgba 1 1 1 1>, EndColor=<rgba 1 0.2422 0 1>, LerpFactor=&UniformAOrBFloat(A=0))
    Lifetime=&RandomRangeFloat(Minimum=0.25, Maximum=0.5)
    Sprite Size Max=<1 4>
    Sprite Size Min=<0.5 3>
    Sprite Size Mode=::NewEnumerator2
    UsePositionOffset=false
    Write Material Random=true
   )
   &SkeletalMeshLocation(
    Bone / Socket Index=$Particles.PartitionParticles.Partition
    Bone Sampling Mode=::NewEnumerator4
    MaskSpawnGroup=false
    Orientation Sampling=::NewEnumerator1
    Skeletal Mesh=$System.Skeletal Mesh
   )
   &AddVelocity(
    Rotation Coordinate Space=::World
    Velocity=&RandomRangeVector(Minimum=<-50 -50 50>, Maximum=<50 50 150>)
   ) {module=AddVelocity001}
   &AddVelocity(
    Rotation Coordinate Space=::World
    Velocity=&TransformVector(
     Vector=&Multiply_Vector(A=$Engine.Owner.Velocity, B=<1 1 0>)
     Source Space=::World
     Destination Space=::Simulation
    )
    Velocity Speed Scale=-0.25
   )
  ]
  particleUpdateScript [
   &ParticleState()
   &ScaleSpriteSize(Uniform Curve Sprite Scale=DataInterface() {propertyValues {Curve {keys [
        {interpMode=RCIM_Cubic, tangentMode=RCTM_User, time=0, value=0}
        {arriveTangent=-0.4008, interpMode=RCIM_Cubic, leaveTangent=-0.4008, tangentMode=RCTM_User, time=0.1, value=1}
        {arriveTangent=-3.0285, interpMode=RCIM_Cubic, leaveTangent=-3.0285, tangentMode=RCTM_User, time=1, value=0}
       ]}}})
   &Drag(Drag=&RandomRangeFloat(Minimum=1, Maximum=2), Use Linear Drag=true, Use Rotational Drag=true)
   &CurlNoiseForce(
    Noise Strength=&Multiply_Float(
     A=350
     B=&FloatFromCurve(FloatCurve=DataInterface() {propertyValues {Curve {keys [<at 0 0>,<at 1 1>]}}})
    )
    Pan Noise=false
   )
   &InheritSourceMovement(
    Applied Position Scale=&Multiply_VectorByFloat(
     Vector=<1 1 1>
     Float=&FloatFromCurve(FloatCurve=DataInterface() {propertyValues {Curve {keys [
          {interpMode=RCIM_Cubic, tangentMode=RCTM_User, time=0, value=1}
          {arriveTangent=-0.3297, interpMode=RCIM_Cubic, leaveTangent=-0.3297, tangentMode=RCTM_User, time=1, value=0.9}
         ]}}})
    )
    ApplyForce=false
    ApplyPosition=true
    ApplyPositionFalloff=false
   )
   &SolveForcesAndVelocity(Clamp Velocity=false, Limit Acceleration=false)
  ]
  rendererClasses [&NiagaraSpriteRendererProperties]
  renderers [{rendererClass=&NiagaraSpriteRendererProperties}]
 }
 NE_FootFlame {
  emitterUpdateScript [
   &EmitterState(
    Inactive Response=::NewEnumerator1
    Life Cycle Mode=::NewEnumerator1
    Loop Behavior=::NewEnumerator1
    Loop Duration Mode=::NewEnumerator1
    UseLoopDelay=false
   )
   &SpawnBurst_Instantaneous(Spawn Count=2, Use Spawn Probability=false, UseLoopCountLimit=false)
  ]
  particleSpawnScript [
   &PartitionParticles()
   &InitializeParticle(
    Color Mode=::NewEnumerator0
    Lifetime=5
    Mesh Scale Mode=::NewEnumerator4
    Sprite Size=<50 50>
    Sprite Size Mode=::NewEnumerator4
    UsePositionOffset=false
    Write Material Random=true
   )
   &SkeletalMeshLocation(
    Bone / Socket Index=$Particles.PartitionParticles.Partition
    Bone Sampling Mode=::NewEnumerator4
    MaskSpawnGroup=false
    Skeletal Mesh=$System.Skeletal Mesh
   )
   /SetVariables() {isSetParameters=true}
   /SetVariables() {isSetParameters=true}
  ]
  particleUpdateScript [
   &ParticleState(Kill Particles When Lifetime Has Elapsed=false)
   /SetVariables(Particles.XOffsetValue:NiagaraFloat=&Add_Float(
     A=$Particles.XOffsetValue
     B=&Multiply_Float(A=$Engine.DeltaTime, B=&Lerp_Float(A=-0.1, B=-0.5, Alpha=$System.SpeedFactor))
    )) {isSetParameters=true}
   &SkeletalMeshLocation(
    Bone / Socket Index=$Particles.PartitionParticles.Partition
    Bone Sampling Mode=::NewEnumerator4
    MaskSpawnGroup=false
    Orientation Sampling=::NewEnumerator1
    Sampled Position Offset=&Lerp_Vector(A=<-3.5 -4 1>, B=<3.5 4 -0.6>, Alpha=&MakeFloatFromInt(Int=$Particles.PartitionParticles.Partition))
    Skeletal Mesh=$System.Skeletal Mesh
   ) {module=SkeletalMeshLocation001}
   /SetVariables(Particles.MeshOrientation=&MultiplyQuaternion(
     Quaternion A=$Particles.SkeletalMeshLocation001.SampledRotation
     Quaternion B=&LerpQuaternion(
      Quaternion A:Quat4f=&MakeQuatFromAxisAngle(Axis=<0 1 0>, AngleInDegrees=90)
      Quaternion B:Quat4f=&MultiplyQuaternion(
       Quaternion A=&MakeQuatFromAxisAngle(Axis=<0 1 0>, AngleInDegrees=-90)
       Quaternion B=&MakeQuatFromAxisAngle(Axis=<0 0 1>, AngleInDegrees=180)
      )
      Lerp Factor:NiagaraFloat=&MakeFloatFromInt(Int=$Particles.PartitionParticles.Partition)
      Enable Spherical Linear Interpolation:NiagaraBool=false
     )
    )) {isSetParameters=true}
   &InterpolateOverTime(
    Rate Of Change=2
    Target Vector Value:Vector3f=&TransformVector(
     Vector=/Niagara/DynamicInputs/Math/Clamp_Vector(
      Vector:Vector3f=&Multiply_VectorByFloat(Vector=$Engine.Owner.Velocity, Float=-0.1)
      Min:Vector3f=<-10 -10 -10>
      Max:Vector3f=<10 10 10>
     )
     Source Space=::Simulation
     Destination Space=::World
    )
    Variable Type=::NewEnumerator2
   )
   &DynamicMaterialParameters(
    Index 0 Param 1=&Add_Float(
     A=&Lerp_Float(A=0, B=0.5, Alpha=&MakeFloatFromInt(Int=$Particles.PartitionParticles.Partition))
     B=$Particles.XOffsetValue
    )
    Index 1 Param 3=&FloatFromCurve(FloatCurve=DataInterface() {propertyValues {Curve {keys [<at 0 0>,<at 1 0.7>]}}}, CurveIndex=$System.SpeedFactor)
    Index 2 Param 1=&MakeFloatFromVector(Vector=$Particles.InterpolateOverTime.InterpolateOverTime_V3_.Moving Average, Channel=::NewEnumerator0)
    Index 2 Param 2=&MakeFloatFromVector(Vector=$Particles.InterpolateOverTime.InterpolateOverTime_V3_.Moving Average, Channel=::NewEnumerator1)
    Index 2 Param 3=&MakeFloatFromVector(Vector=$Particles.InterpolateOverTime.InterpolateOverTime_V3_.Moving Average, Channel=::NewEnumerator2)
    Index 2 Param 4=1
    Module.Index0Param0Write=true
    Module.Index0Param1Write=true
    Module.Index0Param2Write=true
    Module.Index0Param3Write=true
    Module.Index1Param0Write=true
    Module.Index1Param1Write=true
    Module.Index1Param2Write=true
    Module.Index1Param3Write=true
    Module.Index2Param0Write=true
    Module.Index2Param1Write=true
    Module.Index2Param2Write=true
    Module.Index2Param3Write=true
    Write Parameter Index 1=true
    Write Parameter Index 2=true
   )
  ]
  rendererClasses [&NiagaraMeshRendererProperties]
  renderers [{rendererClass=&NiagaraMeshRendererProperties, rendererIndex=0}]
 }
}
summary {
 systemName=NS_RebelClawAviator_Idle_Feet
 emitters [
  {emitterName=NE_FootFlame, rendererClasses [&NiagaraMeshRendererProperties]}
  {emitterName=NE_FootEmbers, rendererClasses [&NiagaraSpriteRendererProperties]}
 ]
 userVariables [
  User.bIsFootHidden {type=NiagaraBool}
  User.bIsFrontEnd {type=NiagaraBool}
  User.bIsFrontEndPreview {type=NiagaraBool}
  User.Debug {type=NiagaraBool}
  User.VariantPicker {type=NiagaraFloat, defaultValue=0, description=""}
 ]
}
