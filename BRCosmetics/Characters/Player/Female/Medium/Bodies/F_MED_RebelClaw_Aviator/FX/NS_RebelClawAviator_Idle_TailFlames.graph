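# Niagara system for the idle tail flames on the Rebel Claw Aviator character, combining mesh flames with sprite embers.
# NE_TailFlames fires four staggered bursts on successive bones from System.BaseBoneSpawn, wrapping modulo 9.
# Flames re-sample their bone each update to stay attached while stretching, sliding and feeding dynamic material parameters.
# NE_TailEmbers spawns near those bones at a speed-scaled, time-gated rate, drifting upward against owner motion with curl noise.
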NiagaraSystem /BRCosmetics/Characters/Player/Female/Medium/Bodies/F_MED_RebelClaw_Aviator/FX/NS_RebelClawAviator_Idle_TailFlames
dependencies {
 usedDataInterfaces [&NiagaraDataInterfaceCurve, &NiagaraDataInterfaceVectorCurve, &NiagaraDataInterfaceSkeletalMesh]
 usedDynamicInputs [&VectorLength, &ClampFloat, &Subtract_Float]
 usedModules [
  &self:SystemSpawnScript.SystemScriptSource.SystemScriptGraph.NiagaraNodeAssignment_7.SetVariables_#1
  &self:SystemSpawnScript.SystemScriptSource.SystemScriptGraph.NiagaraNodeAssignment_0.SetVariables_#2
  &self:SystemSpawnScript.SystemScriptSource.SystemScriptGraph.NiagaraNodeAssignment_5.SetVariables_#3
  &self:SystemSpawnScript.SystemScriptSource.SystemScriptGraph.NiagaraNodeAssignment_6.SetVariables_#4
  &SystemState
  &self:SystemSpawnScript.SystemScriptSource.SystemScriptGraph.NiagaraNodeAssignment_8.SetVariables_#5
  &self:SystemSpawnScript.SystemScriptSource.SystemScriptGraph.NiagaraNodeAssignment_4.SetVariables_#6
  &self:SystemSpawnScript.SystemScriptSource.SystemScriptGraph.NiagaraNodeAssignment_1.SetVariables_#7
  &self:SystemSpawnScript.SystemScriptSource.SystemScriptGraph.NiagaraNodeAssignment_2.SetVariables_#8
  &self:SystemSpawnScript.SystemScriptSource.SystemScriptGraph.NiagaraNodeAssignment_3.SetVariables_#9
 ]
}
emitters {
 NE_TailEmbers {
  emitterUpdateScript [
   &EmitterState()
   &SpawnRate(
    SpawnRate=&MakeCustomFloatFromBool(
     Bool=&SetBoolByFloatComparison(A=$Emitter.LoopedAge, B=0.5, Comparison Type=::NewEnumerator6)
     True Float=&Lerp_Float(A=10, B=15, Alpha=$System.SpeedFactor)
     False Float=0
    )
    Use Spawn Probability=false
   )
  ]
  particleSpawnScript [
   &InitializeParticle(
    Color=&LerpLinearColors(StartColor=<rgba 1 1 1 1>, EndColor=<rgba 1 0.2422 0 1>, LerpFactor=&UniformAOrBFloat(A=0))
    Lifetime=&RandomRangeFloat(Minimum=0.25, Maximum=0.5)
    Sprite Size Max=<1 4>
    Sprite Size Min=<0.5 3>
    Sprite Size Mode=::NewEnumerator2
    UsePositionOffset=false
    Write Material Random=true
   )
   &SkeletalMeshLocation(
    Bone / Socket Index=&Add_Integer(A=$System.BaseBoneSpawn, B=&RandomRangeInt(Minimum=0, Maximum=3))
    Bone Sampling Mode=::NewEnumerator4
    MaskSpawnGroup=false
    Orientation Sampling=::NewEnumerator1
    Skeletal Mesh=$System.Skeletal Mesh
   )
   &AddVelocity(
    Rotation Coordinate Space=::World
    Velocity=&RandomRangeVector(Minimum=<-50 -50 40>, Maximum=<50 50 60>)
   ) {module=AddVelocity001}
   &AddVelocity(
    Rotation Coordinate Space=::World
    Velocity=&TransformVector(
     Vector=&Multiply_Vector(A=$Engine.Owner.Velocity, B=<1 1 0>)
     Source Space=::World
     Destination Space=::Simulation
    )
    Velocity Speed Scale=-0.25
   )
  ]
  particleUpdateScript [
   &ParticleState()
   &ScaleSpriteSize(Uniform Curve Sprite Scale=DataInterface() {propertyValues {Curve {keys [
        {interpMode=RCIM_Cubic, tangentMode=RCTM_User, time=0, value=0}
        {arriveTangent=-0.4008, interpMode=RCIM_Cubic, leaveTangent=-0.4008, tangentMode=RCTM_User, time=0.1, value=1}
        {arriveTangent=-3.0285, interpMode=RCIM_Cubic, leaveTangent=-3.0285, tangentMode=RCTM_User, time=1, value=0}
       ]}}})
   &Drag(Drag=&RandomRangeFloat(Minimum=1, Maximum=2), Use Linear Drag=true, Use Rotational Drag=true)
   &CurlNoiseForce(
    Noise Strength=&Multiply_Float(
     A=350
     B=&FloatFromCurve(FloatCurve=DataInterface() {propertyValues {Curve {keys [<at 0 0>,<at 1 1>]}}})
    )
    Pan Noise=false
   )
   &InheritSourceMovement(
    Applied Position Scale=&Multiply_VectorByFloat(
     Vector=<1 1 1>
     Float=&FloatFromCurve(FloatCurve=DataInterface() {propertyValues {Curve {keys [
          {arriveTangent=-0.0045, interpMode=RCIM_Cubic, leaveTangent=-0.0045, tangentMode=RCTM_User, time=0, value=1}
          {arriveTangent=-0.1402, interpMode=RCIM_Cubic, leaveTangent=-0.1402, tangentMode=RCTM_User, time=1, value=0.9}
         ]}}})
    )
    ApplyForce=false
    ApplyPosition=true
    ApplyPositionFalloff=false
   )
   &SolveForcesAndVelocity(Clamp Velocity=false, Limit Acceleration=false)
  ]
  rendererClasses [&NiagaraSpriteRendererProperties]
  renderers [{rendererClass=&NiagaraSpriteRendererProperties}]
 }
 NE_TailFlames {
  emitterSpawnScript [/SetVariables(Emitter.RandomRoll:NiagaraFloat=&RandomRangeFloat(Minimum=0, Maximum=360)) {isSetParameters=true}]
  emitterUpdateScript [
   &EmitterState()
   &SpawnBurst_Instantaneous(Use Spawn Probability=false, UseLoopCountLimit=false)
   &SpawnBurst_Instantaneous(Spawn Group=1, Spawn Time=0.1, Use Spawn Probability=false, UseLoopCountLimit=false) {module=SpawnBurst_Instantaneous001}
   &SpawnBurst_Instantaneous(Spawn Group=2, Spawn Time=0.2, Use Spawn Probability=false, UseLoopCountLimit=false) {module=SpawnBurst_Instantaneous002}
   &SpawnBurst_Instantaneous(Spawn Group=3, Spawn Time=0.3, Use Spawn Probability=false, UseLoopCountLimit=false) {module=SpawnBurst_Instantaneous003}
  ]
  particleSpawnScript [
   &InitializeParticle(
    Color Mode=::NewEnumerator0
    Lifetime=&RandomRangeFloat(Minimum=0.6, Maximum=0.75)
    Mesh Scale Max=<1 1 4>
    Mesh Scale Min=<1 1 3>
    Mesh Scale Mode=::NewEnumerator2
    Sprite Rotation Mode=::NewEnumerator1
    Sprite Size Mode=::NewEnumerator3
    Uniform Sprite Size=2
    UsePositionOffset=false
    Write Material Random=true
   )
   /SetVariables(Particles.BoneIndex:NiagaraInt32=&ModuloInt(A=&Add_Integer(A=$System.BaseBoneSpawn, B=&SpawnGroup()), B=9)) {isSetParameters=true}
   &SkeletalMeshLocation(
    Bone / Socket Index=$Particles.BoneIndex
    Bone Sampling Mode=::NewEnumerator4
    MaskSpawnGroup=false
    Orientation Sampling=::NewEnumerator1
    Skeletal Mesh=$System.Skeletal Mesh
   )
  ]
  particleUpdateScript [
   &ParticleState()
   &ScaleMeshSize(Scale Factor=&VectorFromCurve(VectorCurve=DataInterface() {propertyValues {XCurve {keys [<at 0 1>,<at 1 0.9>]}, YCurve {keys [<at 0 1>,<at 1 0.9>]}, ZCurve {keys [<at 0 0>,<at 1 1>]}}}))
   &SkeletalMeshLocation(
    Bone / Socket Index=$Particles.BoneIndex
    Bone Sampling Mode=::NewEnumerator4
    MaskSpawnGroup=false
    Orientation Sampling=::NewEnumerator1
    Sampled Position Offset=&Add_Vector(
     A=&Lerp_Vector(A=<0 0 0>, B=<10 0 0>, Alpha=$Particles.NormalizedAge)
     B=&Lerp_Vector(
      A=<0 -1.25 0>
      B=<0 0 0>
      Alpha=&Divide_Float(A=&MakeFloatFromInt(Int=$Particles.BoneIndex), B=8)
     )
    )
    Skeletal Mesh=$System.Skeletal Mesh
   ) {module=SkeletalMeshLocation001}
   &SolveForcesAndVelocity(Clamp Velocity=false, Limit Acceleration=false)
   &DynamicMaterialParameters(
    Index 0 Param 1=&RandomRangeFloat(Minimum=0, Maximum=1)
    Index 1 Param 3=&FloatFromCurve(FloatCurve=DataInterface() {propertyValues {Curve {keys [<at 0 0>,<at 1 1.5>]}}})
    Index 2 Param 1=&MakeFloatFromVector(Vector=$Particles.InterpolateOverTime.InterpolateOverTime_V3_.Moving Average, Channel=::NewEnumerator0)
    Index 2 Param 2=&MakeFloatFromVector(Vector=$Particles.InterpolateOverTime.InterpolateOverTime_V3_.Moving Average, Channel=::NewEnumerator1)
    Index 2 Param 3=&MakeFloatFromVector(Vector=$Particles.InterpolateOverTime.InterpolateOverTime_V3_.Moving Average, Channel=::NewEnumerator2)
    Index 2 Param 4=1
    Module.Index0Param0Write=true
    Module.Index0Param1Write=true
    Module.Index0Param2Write=true
    Module.Index0Param3Write=true
    Module.Index1Param0Write=true
    Module.Index1Param1Write=true
    Module.Index1Param2Write=true
    Module.Index1Param3Write=true
    Module.Index2Param0Write=true
    Module.Index2Param1Write=true
    Module.Index2Param2Write=true
    Module.Index2Param3Write=true
    Write Parameter Index 1=true
    Write Parameter Index 2=true
   )
   /SetVariables(Particles.MeshOrientation=&MultiplyQuaternion(
     Quaternion A=$Particles.SkeletalMeshLocation001.SampledRotation
     Quaternion B=&MultiplyQuaternion(
      Quaternion A=&MakeQuatFromAxisAngle(
       Axis=<1 0 0>
       AngleInDegrees=&Add_Float(A=$Emitter.RandomRoll, B=&RandomRangeFloat(Minimum=0, Maximum=0))
      )
      Quaternion B=&MakeQuatFromAxisAngle(Axis=<0 1 0>, AngleInDegrees=&RandomRangeFloat(Minimum=85, Maximum=95))
     )
    )) {isSetParameters=true}
  ]
  rendererClasses [&NiagaraMeshRendererProperties]
  renderers [{rendererClass=&NiagaraMeshRendererProperties, rendererIndex=0}]
 }
}
summary {
 systemName=NS_RebelClawAviator_Idle_TailFlames
 emitters [
  {emitterName=NE_TailEmbers, rendererClasses [&NiagaraSpriteRendererProperties]}
  {emitterName=NE_TailFlames, rendererClasses [&NiagaraMeshRendererProperties]}
 ]
 userVariables [
  User.bIsFrontEnd {type=NiagaraBool}
  User.bIsFrontEndPreview {type=NiagaraBool}
  User.Debug {type=NiagaraBool}
  User.VariantPicker {type=NiagaraFloat, defaultValue=0, description=""}
 ]
}
