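# Corrective Control Rig for the female RebelClaw Aviator cosmetic body, adjusting knee helper and hood bones as the pose changes.
# A shared function measures each calf's rotation from reference pose; its clamped, remapped yaw weights knee helper offsets.
# A spherical pose reader on head relative to spine_04 sets neckBend_alpha, blended lightly with remapped targeting_alpha to weight hoodmover offsets.
# The reader has debug drawing enabled, and Running_alpha is unused.
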ControlRigBlueprint /BRCosmetics/Characters/Player/Female/Medium/Bodies/F_MED_RebelClaw_Aviator/Meshes/F_MED_RebelClaw_Aviator_CtrlRig
graphs [
 ContainedGraph {nodes [
   n1 = RigVMFunctionEntryNode()
   n2 = &RigUnit_GetTransform(Space=LocalSpace, Item=@n1.Item)
   n3 = &RigUnit_GetTransform(Space=LocalSpace, bInitial=true, Item=@n1.Item)
   > RigVMFunctionReturnNode(OutTransform=/Script/RigVM.RigVMFunction_MathTransformToEulerTransform(Value=&RigVMFunction_MathTransformMakeRelative(Global=@n2.Transform, Parent=@n3.Transform).Local) {is=struct RigVMUnitNode, template="Cast::Execute(in Value,out Result)"}.Result)
  ]}
 RigVMFunctionLibrary {nodes [GetDeltaTransformFromRefPose_FNC = RigVMCollapseNode(ExecuteContext=(), Item=(Type=Bone))]}
 RigVMModel {nodes [
   RigUnit_BeginExecution = &RigUnit_BeginExecution()
   n1 = &RigVMFunction_Sequence(ExecuteContext=@RigUnit_BeginExecution.ExecutePin)
   RigVMFunction_Sequence = &RigVMFunction_Sequence(ExecuteContext=@n1.A)
   GetDeltaTransformFromRefPose_FNC = RigVMFunctionReferenceNode(Item=(Type=Bone,Name="calf_l"), ExecuteContext=@RigVMFunction_Sequence.A)
   > &RigUnit_ModifyTransforms(
    ItemToModify=((Item=(Type=Bone,Name="knee_helper_bottom_l"),Transform=(Rotation=(X=0,Y=0,Z=-0.1736,W=0.9848),Translation=(X=7,Y=0,Z=0),Scale3D=(X=1,Y=1,Z=1))))
    Weight=&RigVMFunction_MathFloatRemap(SourceMaximum=100, SourceMinimum=10, bClamp=true, Value=@GetDeltaTransformFromRefPose_FNC.OutTransform.Rotation.Yaw).Result
   )
   GetDeltaTransformFromRefPose_FNC_1 = RigVMFunctionReferenceNode(Item=(Type=Bone,Name="calf_r"), ExecuteContext=@RigVMFunction_Sequence.B)
   > &RigUnit_ModifyTransforms(
    ItemToModify="((Item=(Type=Bone,Name=\"knee_helper_bottom_r\"),Transform=(Rotation=(X=0,Y=0,Z=0,W=1),Translation=(X=-5,Y=0,Z=0),Scale3D=(X=1,Y=1,Z=1))),(Item=(Type=Bone,Name=\"knee_helper_back_inside_r\"),Transform=(Rotation=(X=0,Y=0,Z=0,W=1),Translation=(X=0,Y=0,Z=10),Scal ...(truncated)"
    Weight=&RigVMFunction_MathFloatRemap(SourceMaximum=100, SourceMinimum=10, bClamp=true, Value=@GetDeltaTransformFromRefPose_FNC_1.OutTransform.Rotation.Yaw).Result
   )
   n2 = &RigVMFunction_Sequence(ExecuteContext=@n1.B)
   n3 = &RigUnit_SphericalPoseReader(
    ActiveRegionScaleFactors=(PositiveWidth=3,NegativeWidth=0.3,PositiveHeight=6,NegativeHeight=6)
    ActiveRegionSize=0.1
    Debug=(bDrawDebug=True,DebugScale=25,DebugSegments=20,DebugThickness=0.25)
    DriverAxis=(X=1,Y=0,Z=0)
    DriverItem=(Type=Bone,Name="head")
    FalloffRegionScaleFactors=(PositiveWidth=-3,NegativeWidth=1,PositiveHeight=0,NegativeHeight=0)
    FalloffSize=0.25
    OptionalParentItem=(Type=Bone,Name="spine_04")
    RotationOffset=(X=90,Y=0,Z=-140)
    ExecutePin=@n2.A
   )
   > RigVMVariableNode(Value=@n3.OutputParam) {variable=neckBend_alpha}
   n4 = RigVMVariableNode() {variable=targeting_alpha}
   &RigUnit_ModifyTransforms(
    ItemToModify="((Item=(Type=Bone,Name=\"hoodmover_b\"),Transform=(Rotation=(X=0,Y=-0.0872,Z=0,W=0.9962),Translation=(X=-0.5,Y=0,Z=0),Scale3D=(X=1,Y=1,Z=1))),(Item=(Type=Bone,Name=\"hoodmover_r\"),Transform=(Rotation=(X=0,Y=0,Z=-0.0872,W=0.9962),Translation=(X=-4.5,Y=0,Z=-5), ...(truncated)"
    Weight=/Script/RigVM.RigVMFunction_MathFloatLerp(
     T=0.2
     B=&RigVMFunction_MathDoubleRemap(SourceMaximum=2, SourceMinimum=0, TargetMaximum=1, TargetMinimum=0, bClamp=False, Value=@n4.Value).Result
     A=RigVMVariableNode() {variable=neckBend_alpha}.Value
    ) {is=struct RigVMUnitNode, template="Interpolate::Execute(in A,in B,in T,out Result)"}.Result
    ExecutePin=@n2.B
   )
  ]}
]
hierarchy [
 root {loc=<0 0 0>, rot=<0 0 0 1>}
 pelvis {parent=root, loc=<0 -0.5301 98.0571>, rot=<0.0127 0.707 -0.0127 -0.707>}
 spine_01 {parent=pelvis, loc=<6.1208 -0.9328 0>, rot=<0 0 0.034 -0.9994>}
 spine_02 {parent=spine_01, loc=<6.5996 -0.2095 0>, rot=<0 0 0.0238 -0.9997>}
 spine_03 {parent=spine_02, loc=<6.4914 -0.1237 0>, rot=<0 0 0.0149 -0.9999>}
 spine_04 {parent=spine_03, loc=<6.9598 0.0794 0>, rot=<0 0 -0.0132 -0.9999>}
 spine_05 {parent=spine_04, loc=<6.9987 0 0>, rot=<0 0 -0.0015 -1>}
 attach_cape {parent=spine_05, loc=<9.3216 -8.4501 0>, rot=<0.0616 -0.7044 -0.0616 -0.7044>}
 neck_01 {parent=spine_05, loc=<11.3138 0.1529 0>, rot=<0 0 -0.2003 -0.9797>}
 neck_02 {parent=neck_01, loc=<4.5627 0 0>, rot=<0 0 0.0504 -0.9987>}
 head {parent=neck_02, loc=<5.3914 0 0>, rot=<0 0 0.115 -0.9934>}
 base_capflap_r {parent=head, loc=<1.0636 1.9294 6.3707>, rot=<-0.1735 -0.9839 0.043 0.0076>}
 dyn_capflap_r {parent=base_capflap_r, loc=<0 0 0>, rot=<0 -0.0323 0 -0.9995>}
 dyn_capflap_lower_r {parent=dyn_capflap_r, loc=<5.6219 0.7819 -0.5525>, rot=<0.0472 -0.0889 -0.0396 -0.9941>}
 base_capflap_l {parent=head, loc=<1.0636 1.9294 -6.6266>, rot=<-0.2164 -0.9763 0.0003 -0.0015>}
 dyn_capflap_l {parent=base_capflap_l, loc=<0 0 0>, rot=<0 0 0 1>}
 dyn_capflap_lower_l {parent=dyn_capflap_l, loc=<5.5905 0.3339 0.3644>, rot=<0.0275 0.1668 -0.0525 -0.9842>}
 base_pigtail_l {parent=head, loc=<-0.0264 -5.581 -5.2>, rot=<0.3466 -0.93 -0.1113 -0.0519>}
 dyn_pigtail_l {parent=base_pigtail_l, loc=<0.0152 -0.0144 0.0832>, rot=<-0.0698 0.0256 0.9957 0.0546>}
 dyn_pigtail_6_l {parent=dyn_pigtail_l, loc=<-2.1269 -2.4118 1.8424>, rot=<0.4321 -0.3464 0.8235 -0.1229>}
 dyn_pigtail_7_l {parent=dyn_pigtail_l, loc=<-3.9872 -0.3019 2.4157>, rot=<0.342 0 0.9397 0>}
 dyn_pigtail_8_l {parent=dyn_pigtail_l, loc=<-4.8967 -2.2482 0.8308>, rot=<0.2981 -0.0392 0.9456 -0.1245>}
 dyn_pigtail_9_l {parent=dyn_pigtail_l, loc=<-5.3804 -0.6518 0.0028>, rot=<-0.0436 0 0.999 0>}
 dyn_pigtail_10_l {parent=dyn_pigtail_l, loc=<-4.0244 0.3492 -0.6495>, rot=<0.0872 0 0.9962 0>}
 dyn_pigtail_12_l {parent=dyn_pigtail_l, loc=<-0.5095 -2.8642 0.6536>, rot=<0.6009 -0.3304 0.7276 0.0192>}
 dyn_pigtail_14_l {parent=dyn_pigtail_l, loc=<-3.3678 -1.0728 -4.3458>, rot=<-0.2588 0 0.9659 0>}
 dyn_pigtail_15_l {parent=dyn_pigtail_l, loc=<-4.7509 -0.0553 -3.4874>, rot=<-0.1736 0 0.9848 0>}
 base_pigtail_r {parent=head, loc=<-0.0264 -5.581 5.2>, rot=<0.3894 -0.9006 0.156 0.1138>}
 dyn_pigtail_r {parent=base_pigtail_r, loc=<0 0 0>, rot=<0 0 0 1>}
 dyn_pigtail_6_r {parent=dyn_pigtail_r, loc=<2.7401 2.9148 -2.0479>, rot=<-0.3464 -0.4321 -0.1229 -0.8235>}
 dyn_pigtail_7_r {parent=dyn_pigtail_r, loc=<4.4671 -0.1614 -2.2428>, rot=<0 -0.342 0 -0.9397>}
 dyn_pigtail_8_r {parent=dyn_pigtail_r, loc=<4.943 2.2515 -0.9079>, rot=<-0.0392 -0.2981 -0.1245 -0.9456>}
 dyn_pigtail_9_r {parent=dyn_pigtail_r, loc=<5.4087 0.6357 -0.0821>, rot=<0 0.0436 0 -0.999>}
 dyn_pigtail_10_r {parent=dyn_pigtail_r, loc=<3.921 0.6131 2.5405>, rot=<0 0.2588 0 -0.9659>}
 dyn_pigtail_11_r {parent=dyn_pigtail_r, loc=<2.2777 -0.3233 3.8216>, rot=<0 0.5 0 -0.866>}
 dyn_pigtail_12_r {parent=dyn_pigtail_r, loc=<0.5378 2.848 -0.733>, rot=<-0.3304 -0.6009 0.0192 -0.7276>}
 attach_backpack {parent=spine_05, loc=<0 -8.35 0>, rot=<-0.0093 -0.707 0.0093 -0.707>}
 clavicle_l {parent=spine_05, loc=<5.4831 -0.1053 -3.2946>, rot=<0.0465 -0.7289 -0.009 -0.6829>}
 upperarm_l {parent=clavicle_l, loc=<9.8658 -1.2136 -0.4105>, rot=<0.0164 -0.3245 -0.0269 -0.9454>}
 lowerarm_l {parent=upperarm_l, loc=<26.465 -0.8345 0.259>, rot=<0.0036 0.0161 -0.2298 -0.9731>}
 hand_l {parent=lowerarm_l, loc=<22.7355 -0.3203 -0.6105>, rot=<-0.6784 -0.0223 0.0735 -0.7307>}
 weapon_l {parent=hand_l, loc=<-0.1974 -3.4111 0.3811>, rot=<0 0 0 1>}
 thumb_01_l {parent=hand_l, loc=<1.3118 -0.3191 -2.459>, rot=<0.6546 -0.4042 0.0664 -0.6354>}
 thumb_02_l {parent=thumb_01_l, loc=<5.9336 -0.0448 0.0401>, rot=<-0.0037 0.0213 0.298 -0.9543>}
 thumb_03_l {parent=thumb_02_l, loc=<2.2223 0.3682 -0.0907>, rot=<0.0118 -0.013 0.0066 -0.9998>}
 index_metacarpal_l {parent=hand_l, loc=<1.6535 0.0183 -2.96>, rot=<0.0158 -0.0124 -0.0173 -0.9996>}
 index_01_l {parent=index_metacarpal_l, loc=<8.3008 0.0357 -0.0524>, rot=<-0.0044 0.0485 0.0924 -0.9945>}
 index_02_l {parent=index_01_l, loc=<3.3587 -0.3004 0.0452>, rot=<-0.002 0.0027 0.3225 -0.9466>}
 index_03_l {parent=index_02_l, loc=<2.5578 -0.123 -0.0252>, rot=<-0.002 0.0067 0.2558 -0.9667>}
 middle_metacarpal_l {parent=hand_l, loc=<2.1856 -0.1725 -0.5984>, rot=<0.006 0.0168 -0.0329 -0.9993>}
 middle_01_l {parent=middle_metacarpal_l, loc=<7.3275 -0.2447 0.0129>, rot=<-0.0208 0.0042 0.0858 -0.9961>}
 middle_02_l {parent=middle_01_l, loc=<4.0123 0.1401 0.1175>, rot=<-0.0218 0.0258 0.3512 -0.9357>}
 middle_03_l {parent=middle_02_l, loc=<2.6077 0.0534 0.1373>, rot=<0.0082 0.0188 0.1708 -0.9851>}
 ring_metacarpal_l {parent=hand_l, loc=<2.2074 -0.3625 1.3638>, rot=<-0.0284 0.0722 -0.0458 -0.9959>}
 ring_01_l {parent=ring_metacarpal_l, loc=<7.0295 -0.6197 -0.1446>, rot=<-0.0489 -0.0252 0.1892 -0.9804>}
 ring_02_l {parent=ring_01_l, loc=<3.6174 0.4647 -0.1027>, rot=<0 0 0.2806 -0.9598>}
 ring_03_l {parent=ring_02_l, loc=<2.4152 0.114 0.0199>, rot=<0 0 0.1189 -0.9929>}
 pinky_metacarpal_l {parent=hand_l, loc=<1.7616 -0.9753 3.3814>, rot=<-0.0397 0.0814 -0.069 -0.9935>}
 pinky_01_l {parent=pinky_metacarpal_l, loc=<7.1709 -0.4831 -0.1094>, rot=<-0.0654 -0.0035 0.2242 -0.9723>}
 pinky_02_l {parent=pinky_01_l, loc=<2.8274 0.0893 -0.1043>, rot=<0.0037 0.0011 0.2854 -0.9584>}
 pinky_03_l {parent=pinky_02_l, loc=<2.2046 0.0307 -0.0078>, rot=<-0.0002 0.0098 0.1445 -0.9895>}
 hand_attach_l {parent=hand_l, loc=<8.5 -3 0>, rot=<0.7071 0.7071 0 0>}
 deform_wrist_dn_base_l {parent=hand_l, loc=<0 -1.5326 0>, rot=<0 0 0 -1>}
 deform_wrist_dn_l {parent=deform_wrist_dn_base_l, loc=<0 0 0>, rot=<0 0 0 1>}
 deform_wrist_up_base_l {parent=hand_l, loc=<0 2.3009 0>, rot=<0 0 0 -1>}
 deform_wrist_up_l {parent=deform_wrist_up_base_l, loc=<0 0 0>, rot=<0 0 0 1>}
 lowerarm_twist_02_l {parent=lowerarm_l, loc=<10.6707 0 0>, rot=<0 0 0 1>}
 lowerarm_twist_01_l {parent=lowerarm_l, loc=<18.316 0 0>, rot=<0 0 0 1>}
 deform_elbow_in_base_l {parent=lowerarm_l, loc=<0.1834 3.6062 -0.2572>, rot=<0 0 0 -1>}
 deform_elbow_in_l {parent=deform_elbow_in_base_l, loc=<0 0 0>, rot=<0 0 0 1>}
 deform_elbow_out_base_l {parent=lowerarm_l, loc=<0 -4 0>, rot=<0 0 0 1>}
 deform_elbow_out_l {parent=deform_elbow_out_base_l, loc=<0 0 0>, rot=<0 0 0 1>}
 upperarm_twist_01_l {parent=upperarm_l, loc=<7.2135 -0.1134 0>, rot=<0 0 0 1>}
 deform_upperarm_bk_base_l {parent=upperarm_twist_01_l, loc=<-6.4053 -4.3746 0.314>, rot=<0 0 0 -1>}
 deform_upperarm_bk_l {parent=deform_upperarm_bk_base_l, loc=<0 0 0>, rot=<0 0 0 1>}
 deform_upperarm_fr_base_l {parent=upperarm_twist_01_l, loc=<-5.6312 5.092 0.3354>, rot=<0 0 0 -1>}
 deform_upperarm_fr_l {parent=deform_upperarm_fr_base_l, loc=<0 0 0>, rot=<0 0 0 1>}
 deform_upperarm_up_base_l {parent=upperarm_twist_01_l, loc=<-8.891 0.4825 3.9517>, rot=<0 0 0 -1>}
 deform_upperarm_up_l {parent=deform_upperarm_up_base_l, loc=<0 0 0>, rot=<0 0 0 1>}
 upperarm_twist_02_l {parent=upperarm_l, loc=<19.7998 -0.1134 0>, rot=<0 0 0 1>}
 hoodmover_l {parent=clavicle_l, loc=<5.8474 -0.5147 4.9625>, rot=<0.1005 0.7235 -0.128 -0.6709>}
 shirtmover_l {parent=clavicle_l, loc=<9.2369 -5.3194 -4.7403>, rot=<0.0164 -0.3245 -0.0269 -0.9454>}
 clavicle_r {parent=spine_05, loc=<5.4831 -0.1053 3.2946>, rot=<0.7289 0.0465 -0.6829 0.009>}
 upperarm_r {parent=clavicle_r, loc=<-9.8658 1.2136 0.4105>, rot=<0.0164 -0.3245 -0.0269 -0.9454>}
 lowerarm_r {parent=upperarm_r, loc=<-26.465 0.8345 -0.259>, rot=<0.0036 0.0161 -0.2298 -0.9731>}
 hand_r {parent=lowerarm_r, loc=<-22.7355 0.3203 0.6106>, rot=<-0.6784 -0.0223 0.0735 -0.7307>}
 weapon_r {parent=hand_r, loc=<0.1974 3.4111 -0.3811>, rot=<0 0 0 1>}
 thumb_01_r {parent=hand_r, loc=<-1.3118 0.3191 2.459>, rot=<0.6546 -0.4042 0.0664 -0.6354>}
 thumb_02_r {parent=thumb_01_r, loc=<-5.9336 0.0448 -0.0401>, rot=<-0.0037 0.0213 0.298 -0.9543>}
 thumb_03_r {parent=thumb_02_r, loc=<-2.2223 -0.3682 0.0907>, rot=<0.0118 -0.013 0.0066 -0.9998>}
 index_metacarpal_r {parent=hand_r, loc=<-1.6535 -0.0183 2.96>, rot=<0.0158 -0.0124 -0.0173 -0.9996>}
 index_01_r {parent=index_metacarpal_r, loc=<-8.3008 -0.0357 0.0524>, rot=<-0.0044 0.0485 0.0924 -0.9945>}
 index_02_r {parent=index_01_r, loc=<-3.3587 0.3004 -0.0452>, rot=<-0.002 0.0027 0.3225 -0.9466>}
 index_03_r {parent=index_02_r, loc=<-2.5578 0.123 0.0252>, rot=<-0.002 0.0067 0.2558 -0.9667>}
 middle_metacarpal_r {parent=hand_r, loc=<-2.1856 0.1725 0.5984>, rot=<0.006 0.0168 -0.0329 -0.9993>}
 middle_01_r {parent=middle_metacarpal_r, loc=<-7.3275 0.2447 -0.0129>, rot=<-0.0208 0.0042 0.0858 -0.9961>}
 middle_02_r {parent=middle_01_r, loc=<-4.0123 -0.1401 -0.1175>, rot=<-0.0218 0.0258 0.3512 -0.9357>}
 middle_03_r {parent=middle_02_r, loc=<-2.6077 -0.0534 -0.1373>, rot=<0.0082 0.0188 0.1708 -0.9851>}
 ring_metacarpal_r {parent=hand_r, loc=<-2.2074 0.3625 -1.3638>, rot=<-0.0284 0.0722 -0.0458 -0.9959>}
 ring_01_r {parent=ring_metacarpal_r, loc=<-7.0295 0.6197 0.1446>, rot=<-0.0489 -0.0252 0.1892 -0.9804>}
 ring_02_r {parent=ring_01_r, loc=<-3.6174 -0.4647 0.1027>, rot=<0 0 0.2806 -0.9598>}
 ring_03_r {parent=ring_02_r, loc=<-2.4152 -0.114 -0.0199>, rot=<0 0 0.1189 -0.9929>}
 pinky_metacarpal_r {parent=hand_r, loc=<-1.7616 0.9753 -3.3814>, rot=<-0.0397 0.0814 -0.069 -0.9935>}
 pinky_01_r {parent=pinky_metacarpal_r, loc=<-7.1709 0.4831 0.1094>, rot=<-0.0654 -0.0035 0.2242 -0.9723>}
 pinky_02_r {parent=pinky_01_r, loc=<-2.8274 -0.0893 0.1043>, rot=<0.0037 0.0011 0.2854 -0.9584>}
 pinky_03_r {parent=pinky_02_r, loc=<-2.2046 -0.0307 0.0078>, rot=<-0.0002 0.0098 0.1445 -0.9895>}
 hand_attach_r {parent=hand_r, loc=<-8.5001 2.9999 0>, rot=<0 0 -0.7071 -0.7071>}
 deform_wrist_dn_base_r {parent=hand_r, loc=<-0.0001 1.5328 0>, rot=<0 0 0 -1>}
 deform_wrist_dn_r {parent=deform_wrist_dn_base_r, loc=<0 0 0>, rot=<0 0 0 1>}
 deform_wrist_up_base_r {parent=hand_r, loc=<-0.0002 -2.3007 -0.0001>, rot=<0 0 0 -1>}
 deform_wrist_up_r {parent=deform_wrist_up_base_r, loc=<0 0 0>, rot=<0 0 0 1>}
 lowerarm_twist_02_r {parent=lowerarm_r, loc=<-10.6707 0 0>, rot=<0 0 0 -1>}
 lowerarm_twist_01_r {parent=lowerarm_r, loc=<-18.316 0 0>, rot=<0 0 0 -1>}
 deform_elbow_in_base_r {parent=lowerarm_r, loc=<-0.1834 -3.6062 0.2572>, rot=<0 0 0 -1>}
 deform_elbow_in_r {parent=deform_elbow_in_base_r, loc=<0 0 0>, rot=<0 0 0 1>}
 deform_elbow_out_base_r {parent=lowerarm_r, loc=<0 4 -0.0001>, rot=<0 0 0 -1>}
 deform_elbow_out_r {parent=deform_elbow_out_base_r, loc=<0 0 0>, rot=<0 0 0 1>}
 upperarm_twist_01_r {parent=upperarm_r, loc=<-7.2135 0.1134 0>, rot=<0 0 0 -1>}
 deform_upperarm_bk_base_r {parent=upperarm_twist_01_r, loc=<6.4053 4.3746 -0.314>, rot=<0 0 0 -1>}
 deform_upperarm_bk_r {parent=deform_upperarm_bk_base_r, loc=<0 0 0>, rot=<0 0 0 1>}
 deform_upperarm_fr_base_r {parent=upperarm_twist_01_r, loc=<5.6312 -5.092 -0.3354>, rot=<0 0 0 -1>}
 deform_upperarm_fr_r {parent=deform_upperarm_fr_base_r, loc=<0 0 0>, rot=<0 0 0 1>}
 deform_upperarm_up_base_r {parent=upperarm_twist_01_r, loc=<8.891 -0.4825 -3.9517>, rot=<0 0 0 -1>}
 deform_upperarm_up_r {parent=deform_upperarm_up_base_r, loc=<0 0 0>, rot=<0 0 0 1>}
 upperarm_twist_02_r {parent=upperarm_r, loc=<-19.7998 0.1134 0>, rot=<0 0 0 -1>}
 hoodmover_r {parent=clavicle_r, loc=<-5.8474 0.5147 -4.9625>, rot=<0.7235 -0.1005 -0.6709 0.128>}
 shirtmover_r {parent=clavicle_r, loc=<-9.2369 5.3194 4.7403>, rot=<0.0164 -0.3245 -0.0269 -0.9454>}
 hoodmover_b {parent=spine_05, loc=<10.7374 -8.5713 0>, rot=<0 0 -0.1091 -0.994>}
 base_ear_l {parent=spine_05, loc=<6.0022 -8.952 -5.8224>, rot=<-0.2449 -0.0931 0.9477 -0.1822>}
 dyn_ear_l {parent=base_ear_l, loc=<0 0 0>, rot=<0 0 0 1>}
 base_ear_r {parent=spine_05, loc=<6.0017 -8.952 5.8224>, rot=<0.2121 0.0433 0.9565 -0.1955>}
 dyn_ear_r {parent=base_ear_r, loc=<0 0 0>, rot=<0 0 0 1>}
 base_hoodiestring_r {parent=spine_05, loc=<6.6397 6.4623 7.814>, rot=<-0.0653 0.0738 -0.8445 -0.5265>}
 dyn_hoodiestring_r {parent=base_hoodiestring_r, loc=<0 0 0>, rot=<0 0 0 1>}
 dyn_hoodiestring_r1 {parent=dyn_hoodiestring_r, loc=<4.7588 3.5193 -0.2877>, rot=<0 0 0 1>}
 dyn_hoodiestring_r2 {parent=dyn_hoodiestring_r1, loc=<4.0713 2.7598 -0.2607>, rot=<0 0 0 1>}
 base_hoodiestring_l {parent=spine_05, loc=<6.6395 6.4623 -7.814>, rot=<0.0653 -0.0738 -0.8445 -0.5265>}
 dyn_hoodiestring_l {parent=base_hoodiestring_l, loc=<0 0 0>, rot=<0 0 0 1>}
 dyn_hoodiestring_l1 {parent=dyn_hoodiestring_l, loc=<4.7591 3.5195 0.288>, rot=<0 0 0 1>}
 dyn_hoodiestring_l2 {parent=dyn_hoodiestring_l1, loc=<4.0716 2.7606 0.2607>, rot=<0 0 -0.1736 -0.9848>}
 base_necklace {parent=spine_05, loc=<10.8944 0.1215 0.0127>, rot=<-0.039 -0.0196 -0.7138 -0.699>}
 dyn_necklace {parent=base_necklace, loc=<0 0 0>, rot=<0.0807 -0.1337 -0.0544 -0.9862>}
 dyn_base_tag {parent=dyn_necklace, loc=<10.1718 6.876 0.7933>, rot=<0.0504 -0.0062 0.0256 -0.9984>}
 dyn_tag_bottom {parent=dyn_base_tag, loc=<0.1179 0.487 0.2789>, rot=<0 0 -0.0436 -0.999>}
 dyn_tag_top {parent=dyn_base_tag, loc=<0.5543 0.4606 -0.1951>, rot=<0 0 -0.0436 -0.999>}
 deform_pec_aim_base_l {parent=spine_04, loc=<3.2456 9.3309 -1.2237>, rot=<0.0603 -0.4033 0.3318 -0.8507>}
 deform_pec_aim_l {parent=deform_pec_aim_base_l, loc=<0 0 0>, rot=<0 0 0 1>}
 deform_pec_base_l {parent=deform_pec_aim_l, loc=<9.3665 0 0>, rot=<0 0 0 1>}
 deform_pec_l {parent=deform_pec_base_l, loc=<0 0 0>, rot=<0 0 0 1>}
 deform_pec_aim_base_r {parent=spine_04, loc=<3.2452 9.3309 1.2237>, rot=<-0.4033 -0.0603 0.8507 0.3318>}
 deform_pec_aim_r {parent=deform_pec_aim_base_r, loc=<0 0 0>, rot=<0 0 0 1>}
 deform_pec_base_r {parent=deform_pec_aim_r, loc=<-9.3667 -0.0003 -0.0004>, rot=<0 0 0 1>}
 deform_pec_r {parent=deform_pec_base_r, loc=<0 0 0>, rot=<0 0 0 1>}
 thigh_l {parent=pelvis, loc=<-1.0855 -0.2109 -8.7053>, rot=<0.0009 0.0355 0.0155 -0.9993>}
 calf_l {parent=thigh_l, loc=<-42.5284 -0.0019 -0.7297>, rot=<0.0006 -0.0161 -0.0391 -0.9991>}
 foot_l {parent=calf_l, loc=<-46.7401 -2.5014 -0.592>, rot=<0.0427 -0.0238 0.0353 -0.9982>}
 ball_l {parent=foot_l, loc=<-3.5658 10.8066 -0.1544>, rot=<0.0001 -0.0001 -0.7063 -0.7079>}
 calf_twist_02_l {parent=calf_l, loc=<-19.1546 -1.7943 -0.2638>, rot=<0.0237 0.0095 -0.0035 -0.9997>}
 calf_twist_01_l {parent=calf_l, loc=<-35.8483 -1.9197 -0.5794>, rot=<0.0237 0.0095 -0.0035 -0.9997>}
 base_bootloop_l {parent=calf_twist_01_l, loc=<3.7589 -6.3548 -0.0042>, rot=<-0.0102 -0.0381 0.4737 -0.8798>}
 dyn_bootloop_l {parent=base_bootloop_l, loc=<0 0 0>, rot=<0 0 0 1>}
 base_boottongue_l {parent=calf_twist_01_l, loc=<-2.677 5.0246 -0.0244>, rot=<-0.0302 -0.0254 -0.1166 -0.9924>}
 dyn_boottongue_l {parent=base_boottongue_l, loc=<0 0 0>, rot=<0 0 0 1>}
 knee_helper_l {parent=calf_l, loc=<1.3294 5.1376 0.8783>, rot=<0 0 0 1>}
 knee_helper_bottom_l {parent=calf_l, loc=<-7.2159 3.4842 0.4002>, rot=<0 0 0 1>}
 deform_knee_in_base_l {parent=calf_l, loc=<0.8031 -5.0231 0.2456>, rot=<0 0 0 -1>}
 deform_knee_in_l {parent=deform_knee_in_base_l, loc=<0 0 0>, rot=<0 0 0 1>}
 deform_knee_out_base_l {parent=calf_l, loc=<0 4 0>, rot=<0 0 0 -1>}
 deform_knee_out_l {parent=deform_knee_out_base_l, loc=<0 0 0>, rot=<0 0 0 1>}
 thigh_twist_01_l {parent=thigh_l, loc=<-27.952 0 0>, rot=<0.0013 0 0 -1>}
 deform_groin_base_l {parent=thigh_l, loc=<0.4801 5.6776 -0.2486>, rot=<0 0 0 -1>}
 deform_groin_l {parent=deform_groin_base_l, loc=<0 0 0>, rot=<0 0 0 1>}
 thigh_r {parent=pelvis, loc=<-1.0855 -0.2109 8.7053>, rot=<0.0355 -0.0009 0.9993 0.0155>}
 calf_r {parent=thigh_r, loc=<42.5284 0.0019 0.7297>, rot=<0.0006 -0.0161 -0.0391 -0.9991>}
 foot_r {parent=calf_r, loc=<46.7401 2.5014 0.592>, rot=<0.0427 -0.0238 0.0353 -0.9982>}
 ball_r {parent=foot_r, loc=<3.5658 -10.8066 0.1544>, rot=<0.0001 -0.0001 -0.7063 -0.7079>}
 calf_twist_02_r {parent=calf_r, loc=<19.1546 1.7943 0.2638>, rot=<0.0237 0.0095 -0.0035 -0.9997>}
 calf_twist_01_r {parent=calf_r, loc=<35.8483 1.9197 0.5794>, rot=<0.0237 0.0095 -0.0035 -0.9997>}
 base_bootloop_r {parent=calf_twist_01_r, loc=<-3.7589 6.3548 0.0042>, rot=<0.0381 -0.0102 0.8798 0.4737>}
 dyn_bootloop_r {parent=base_bootloop_r, loc=<0 0 0>, rot=<0 0 0 1>}
 base_boottongue_r {parent=calf_twist_01_r, loc=<2.677 -5.0246 0.0248>, rot=<0.0254 -0.0302 0.9924 -0.1166>}
 dyn_boottongue_r {parent=base_boottongue_r, loc=<0 0 0>, rot=<0 0 0 1>}
 knee_helper_r {parent=calf_r, loc=<-1.329 -5.138 -0.878>, rot=<0.0385 -0.0063 0.9992 0>}
 knee_helper_bottom_r {parent=calf_r, loc=<7.2272 -3.4864 -0.1163>, rot=<0 0 0 1>}
 knee_helper_back_outside_r {parent=calf_r, loc=<5.0725 2.4086 3.1677>, rot=<0 0 0 1>}
 knee_helper_back_inside_r {parent=calf_r, loc=<4.982 3.8157 -2.1349>, rot=<0 0 0 1>}
 deform_knee_in_base_r {parent=calf_r, loc=<-0.8031 5.0231 -0.2457>, rot=<0 0 0 -1>}
 deform_knee_in_r {parent=deform_knee_in_base_r, loc=<0 0 0>, rot=<0 0 0 1>}
 deform_knee_out_base_r {parent=calf_r, loc=<0 -4 0>, rot=<0 0 0 -1>}
 deform_knee_out_r {parent=deform_knee_out_base_r, loc=<0 0 0>, rot=<0 0 0 1>}
 thigh_twist_01_r {parent=thigh_r, loc=<27.952 0 0>, rot=<0.0013 0 0 -1>}
 deform_groin_base_r {parent=thigh_r, loc=<-0.4801 -5.6776 0.2486>, rot=<0 0 0 -1>}
 deform_groin_r {parent=deform_groin_base_r, loc=<0 0 0>, rot=<0 0 0 1>}
 deform_glute_base_l {parent=pelvis, loc=<-7.3466 -9.0213 -9.1245>, rot=<0.0172 0.031 -0.4477 -0.8935>}
 deform_glute_l {parent=deform_glute_base_l, loc=<0 0 0>, rot=<0 0 0 1>}
 deform_glute_base_r {parent=pelvis, loc=<-7.3464 -9.0213 9.1245>, rot=<0.031 -0.0172 0.8935 -0.4477>}
 deform_glute_r {parent=deform_glute_base_r, loc=<0 0 0>, rot=<0 0 0 1>}
 attach {parent=root, loc=<0 0 0>, rot=<0 0 0 1>}
 ik_hand_root {parent=root, loc=<0 0 0>, rot=<0 0 0 1>}
 ik_hand_gun {parent=ik_hand_root, loc=<-48.1743 3.3545 104.4381>, rot=<0.6276 0.1309 0.4183 -0.6434>}
 ik_hand_r {parent=ik_hand_gun, loc=<0 0 0>, rot=<0 0 0 1>}
 ik_hand_l {parent=ik_hand_gun, loc=<59.3271 67.6997 34.3524>, rot=<-0.6276 -0.1309 -0.4183 -0.6434>}
 ik_foot_root {parent=root, loc=<0 0 0>, rot=<0 0 0 1>}
 ik_foot_l {parent=ik_foot_root, loc=<14.8262 -7.3545 8.296>, rot=<0.0357 0.7045 0.0284 -0.7082>}
 ik_foot_r {parent=ik_foot_root, loc=<-14.8262 -7.3545 8.296>, rot=<0.7082 0.0284 -0.7045 0.0357>}
]
variables [
 Running_alpha {type=double, default=0}
 neckBend_alpha {type=double, default=0}
 targeting_alpha {type=double, default=0}
]
