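# Physics asset for the female medium RebelClaw Aviator body, rigging its hoodie drawstrings and tagged necklace for physics-driven motion.
# Capsule chains hang from spine_05: one per drawstring, plus a necklace branching into top and bottom tags.
# Tight swing limits and locked or near-locked twist constrain the chains, with the right string base looser than the left.
# Chest and upper-arm capsules add body volumes, the arms linked to spine_05 by free constraints.
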PhysicsAsset /BRCosmetics/Characters/Player/Female/Medium/Bodies/F_MED_RebelClaw_Aviator/Meshes/F_MED_RebelClaw_Aviator_Physics
bodies {
 base_hoodiestring_l [{shapeType=Capsule, length=0.505, radius=0.505, shapeName=base_hoodiestring_l_capsule, rotation=<rot 0 -90 0>}]
 base_hoodiestring_r [{shapeType=Capsule, length=0.505, radius=0.505, shapeName=base_hoodiestring_r_capsule, rotation=<rot 0 -90 0>}]
 base_necklace [{shapeType=Capsule, length=1, radius=1, shapeName=base_necklace_capsule}]
 dyn_base_tag [{shapeType=Capsule, length=1, radius=1, shapeName=dyn_base_tag_capsule}]
 dyn_hoodiestring_l [{
   shapeType=Capsule
   length=1.1392
   radius=0.6706
   shapeName=dyn_hoodiestring_l_capsule
   center=<4.1807 3.1688 0.2926>
   rotation=<rot 0 123.2471 -86.0217>
  }]
 dyn_hoodiestring_l1 [{
   shapeType=Capsule
   length=0.9906
   radius=0.6564
   shapeName=dyn_hoodiestring_l1_capsule
   center=<3.0894 2.1439 0.2738>
   rotation=<rot 0 128.6954 -87.2768>
  }]
 dyn_hoodiestring_l2 [{
   shapeType=Capsule
   length=1.6357
   radius=0.6148
   shapeName=dyn_hoodiestring_l2_capsule
   center=<2.3537 2.0262 0.1003>
   rotation=<rot 0 135.4778 -89.5359>
  }]
 dyn_hoodiestring_r [{
   shapeType=Capsule
   length=1.1392
   radius=0.6706
   shapeName=dyn_hoodiestring_r_capsule
   center=<4.1808 3.1689 -0.2926>
   rotation=<rot 0 -56.7529 -86.0217>
  }]
 dyn_hoodiestring_r1 [{
   shapeType=Capsule
   length=0.9906
   radius=0.6564
   shapeName=dyn_hoodiestring_r1_capsule
   center=<3.0899 2.1443 -0.2741>
   rotation=<rot 0 -51.3046 -87.2768>
  }]
 dyn_hoodiestring_r2 [{
   shapeType=Capsule
   length=1.6356
   radius=0.6148
   shapeName=dyn_hoodiestring_r2_capsule
   center=<1.5195 2.7102 -0.1006>
   rotation=<rot 0 -24.5222 -89.5359>
  }]
 dyn_necklace [{
   shapeType=Capsule
   length=3.8392
   radius=1.9495
   shapeName=dyn_necklace_capsule
   center=<6.7497 5.3036 1.9152>
   rotation=<rot 90 -74.3018 -104.6866>
  }]
 dyn_tag_bottom [{shapeType=Capsule, length=1, radius=1, shapeName=dyn_tag_bottom_capsule, center=<0.5546 0.6174 0>, rotation=<rot 90 0 -56.8549>}]
 dyn_tag_top [{shapeType=Capsule, length=1, radius=1, shapeName=dyn_tag_top_capsule, center=<0.5546 0.6174 0>, rotation=<rot 90 0 -56.8549>}]
 spine_05 [{shapeType=Capsule, length=26.3284, radius=8.1306, shapeName=spine_05_capsule, center=<-5.2136 3.6871 0>}]
 upperarm_twist_01_l [{shapeType=Capsule, length=20, radius=4.5, shapeName=upperarm_twist_01_l_capsule, rotation=<rot 90 0 0>}]
 upperarm_twist_01_r [{shapeType=Capsule, length=20, radius=4.5, shapeName=upperarm_twist_01_r_capsule, rotation=<rot 90 0 0>}]
}
constraints [
 {bone1Name=base_hoodiestring_l, bone2Name=spine_05, swing1LimitDegrees=10, swing2LimitDegrees=10, twistLimitDegrees=0}
 {bone1Name=dyn_hoodiestring_l, bone2Name=base_hoodiestring_l, swing1LimitDegrees=20, swing2LimitDegrees=10, twistLimitDegrees=0}
 {bone1Name=dyn_hoodiestring_l1, bone2Name=dyn_hoodiestring_l, swing1LimitDegrees=20, swing2LimitDegrees=20, twistLimitDegrees=0}
 {bone1Name=dyn_hoodiestring_l2, bone2Name=dyn_hoodiestring_l1, swing1LimitDegrees=10, swing2LimitDegrees=10, twistLimitDegrees=0}
 {bone1Name=base_hoodiestring_r, bone2Name=spine_05, swing1LimitDegrees=45, swing2LimitDegrees=45, twistLimitDegrees=0}
 {bone1Name=dyn_hoodiestring_r, bone2Name=base_hoodiestring_r, swing1LimitDegrees=20, swing2LimitDegrees=10, twistLimitDegrees=0}
 {bone1Name=dyn_hoodiestring_r1, bone2Name=dyn_hoodiestring_r, swing1LimitDegrees=20, swing2LimitDegrees=20, twistLimitDegrees=0}
 {bone1Name=dyn_hoodiestring_r2, bone2Name=dyn_hoodiestring_r1, swing1LimitDegrees=10, swing2LimitDegrees=10, twistLimitDegrees=0}
 {bone1Name=base_necklace, bone2Name=spine_05, swing1LimitDegrees=0, swing2LimitDegrees=45, twistLimitDegrees=0}
 {bone1Name=dyn_necklace, bone2Name=base_necklace, swing1LimitDegrees=0, swing2LimitDegrees=15, twistLimitDegrees=0}
 {bone1Name=dyn_base_tag, bone2Name=dyn_necklace, swing1LimitDegrees=10, swing2LimitDegrees=0, twistLimitDegrees=1}
 {bone1Name=dyn_tag_bottom, bone2Name=dyn_base_tag, swing1LimitDegrees=0, swing2LimitDegrees=50, twistLimitDegrees=1}
 {bone1Name=dyn_tag_top, bone2Name=dyn_base_tag, swing1LimitDegrees=0, swing2LimitDegrees=50, twistLimitDegrees=1}
 {
  bone1Name=upperarm_twist_01_l
  bone2Name=spine_05
  swing1LimitDegrees=45
  swing1Motion=Free
  swing2LimitDegrees=45
  swing2Motion=Free
  twistLimitDegrees=45
  twistMotion=Free
 }
 {
  bone1Name=upperarm_twist_01_r
  bone2Name=spine_05
  swing1LimitDegrees=45
  swing1Motion=Free
  swing2LimitDegrees=45
  swing2Motion=Free
  twistLimitDegrees=45
  twistMotion=Free
 }
]
