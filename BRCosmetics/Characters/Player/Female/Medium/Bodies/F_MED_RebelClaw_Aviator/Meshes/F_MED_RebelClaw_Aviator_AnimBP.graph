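# Cosmetic AnimBP for the RebelClaw Aviator female body, adding secondary motion and pose fixes to the pose copied from its attached parent.
# Two Control Rigs feed shirt constraints, RigidBody physics, AnimDynamics on boots and ears, then additive hood and drawstring offsets.
# The event graph derives blend alphas from movement, targeting, skydiving, arm angles and weapon type.
# On customization load, it hides any tail mesh under long back bling.
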AnimBlueprint /BRCosmetics/Characters/Player/Female/Medium/Bodies/F_MED_RebelClaw_Aviator/Meshes/F_MED_RebelClaw_Aviator_AnimBP
target_skeleton=&/F_MED_RebelClaw_Aviator_Skeleton
graphs [
 AnimGraph {nodes [
   n1 = Anim Dynamics - Bone: dyn_ear_l(ComponentPose=Local To Component(LocalPose=Use cached pose 'bootDyn'()), Alpha=1, ExternalForce="0,0,0") {
    angularDampingOverride=0.999
    bGravityOverrideInSimSpace=true
    bOverrideAngularDamping=true
    bUseGravityOverride=true
    alphaBoolBlend {blendOption=Linear}
    boundBone {boneName=dyn_ear_l}
    componentAppliedLinearAccClamp=<3000 3000 3000>
    componentLinearAccScale=<0.3 0.3 0.3>
    componentLinearVelScale=<0.3 0.3 0.3>
    gravityOverride {z=-980}
    physicsBodyDefinitions [{
      sphereCollisionRadius=0
      boundBone {boneName=dyn_ear_l}
      boxExtents=<10 12 12>
      constraintSetup {angularLimitsMax=<10 10 10>, angularLimitsMin=<0 0 0>, angularTarget=<0 0 0>}
      localJointOffset=<2 0.2 0.2>
     }]
   }
   n2 = Anim Dynamics - Bone: dyn_ear_r(ComponentPose=@n1, Alpha=1, ExternalForce="0,0,0") {
    angularDampingOverride=0.999
    bGravityOverrideInSimSpace=true
    bOverrideAngularDamping=true
    bUseGravityOverride=true
    alphaBoolBlend {blendOption=Linear}
    boundBone {boneName=dyn_ear_r}
    componentAppliedLinearAccClamp=<3000 3000 3000>
    componentLinearAccScale=<0.3 0.3 0.3>
    componentLinearVelScale=<0.3 0.3 0.3>
    gravityOverride {z=-980}
    physicsBodyDefinitions [{
      sphereCollisionRadius=0
      boundBone {boneName=dyn_ear_r}
      boxExtents=<10 12 12>
      constraintSetup {angularLimitsMax=<10 10 10>, angularLimitsMin=<0 0 0>, angularTarget=<0 0 0>}
      localJointOffset=<2 0.2 0.2>
     }]
   }
   n3 = "Transform (Modify) Bone - Bone: base_hoodiestring_r"(
    ComponentPose=@n2
    Alpha=Get String_r_out() -> (String_r_out:double)
    Translation="0.960,-1.864,0.008"
    Rotation="0.062,1.044,0.354"
    Scale="1,1,1"
   ) {
    rotationMode=BMM_Additive
    rotationSpace=BCS_BoneSpace
    translationMode=BMM_Additive
    translationSpace=BCS_BoneSpace
    alphaBoolBlend {blendOption=Linear}
    boneToModify {boneName=base_hoodiestring_r}
    rotation=<rot 0.062 1.044 0.354>
    translation=<0.9604 -1.864 0.008>
   }
   n4 = "Transform (Modify) Bone - Bone: base_hoodiestring_l"(ComponentPose=@n3, Alpha=Get Strings out() -> (Strings out:double), Translation="1,-1.8,0", Rotation="-0.193,9.682,0.020", Scale="1,1,1") {
    rotationMode=BMM_Additive
    rotationSpace=BCS_BoneSpace
    translationMode=BMM_Additive
    translationSpace=BCS_BoneSpace
    alphaBoolBlend {blendOption=Linear}
    boneToModify {boneName=base_hoodiestring_l}
    rotation=<rot -0.193 9.682 0.02>
    translation=<1 -1.8>
   }
   n5 = Get isSkydiving() -> (isSkydiving:double)
   n6 = "Transform (Modify) Bone - Bone: hoodmover_r"(ComponentPose=@n4, Alpha=@n5, Translation="-1.677,0.710,4.975", Rotation="17.628,13.610,38.605", Scale="1,1,1") {
    rotationMode=BMM_Additive
    rotationSpace=BCS_BoneSpace
    translationMode=BMM_Additive
    translationSpace=BCS_BoneSpace
    alphaBoolBlend {blendOption=Linear}
    alphaScaleBiasClamp {bInterpResult=true, interpSpeedDecreasing=1, interpSpeedIncreasing=1}
    boneToModify {boneName=hoodmover_r}
    rotation=<rot 17.628 13.61 38.605>
    translation=<-1.6771 0.71 4.975>
   }
   n7 = "Transform (Modify) Bone - Bone: hoodmover_l"(ComponentPose=@n6, Alpha=@n5, Translation="-3.926,-1.712,-5.162", Rotation="15.872,-13.243,-6.470", Scale="1,1,1") {
    rotationMode=BMM_Additive
    rotationSpace=BCS_BoneSpace
    translationMode=BMM_Additive
    translationSpace=BCS_BoneSpace
    alphaBoolBlend {blendOption=Linear}
    alphaScaleBiasClamp {bInterpResult=true, interpSpeedDecreasing=1, interpSpeedIncreasing=1}
    boneToModify {boneName=hoodmover_l}
    rotation=<rot 15.872 -13.243 -6.47>
    translation=<-3.926 -1.712 -5.162>
   }
   n8 = "Transform (Modify) Bone - Bone: hoodmover_b"(
    ComponentPose=@n7
    Alpha=Get SkydivingFloating_PointingDown() -> (SkydivingFloating_PointingDown:double)
    Translation="-1.714,-0.509,0.121"
    Rotation="-9.435,-0.000,0.000"
    Scale="0.800,1.000,1.545"
   ) {
    rotationMode=BMM_Additive
    rotationSpace=BCS_BoneSpace
    scaleMode=BMM_Additive
    scaleSpace=BCS_BoneSpace
    translationMode=BMM_Additive
    translationSpace=BCS_BoneSpace
    alphaBoolBlend {blendOption=Linear}
    alphaScaleBiasClamp {bInterpResult=true}
    boneToModify {boneName=hoodmover_b}
    rotation=<rot -9.435 -0.0001 0>
    scale {x=0.8, z=1.545}
    translation=<-1.714 -0.509 0.121>
   }
   Output Pose(Result=Component To Local(ComponentPose=@n8)) {layerGroup=None}
   n9 = Anim Dynamics - Bone: dyn_boottongue_l(ComponentPose=Local To Component(LocalPose=Use cached pose 'startpose'()), Alpha=1, ExternalForce="0,0,0") {
    angularDampingOverride=0.1
    angularSpringConstant=100
    bAngularSpring=true
    bGravityOverrideInSimSpace=true
    bUseGravityOverride=true
    simulationSpace=BoneRelative
    alphaBoolBlend {blendOption=Linear}
    boundBone {boneName=dyn_boottongue_l}
    componentAppliedLinearAccClamp=<3000 3000 3000>
    componentLinearAccScale=<0.3 0.3 0.3>
    componentLinearVelScale=<0.3 0.3 0.3>
    gravityOverride {z=-980}
    physicsBodyDefinitions [{
      sphereCollisionRadius=0
      boundBone {boneName=dyn_boottongue_l}
      boxExtents=<6 7 7>
      constraintSetup {angularLimitsMax=<5 0 40>, angularLimitsMin=<0 0 -5>, angularTarget=<5 0 0>}
      localJointOffset=<4 0.2 0.2>
     }]
    relativeSpaceBone {boneName=calf_twist_01_l}
   }
   n10 = Anim Dynamics - Bone: dyn_bootloop_l(ComponentPose=@n9, Alpha=1, ExternalForce="0,0,0") {
    angularDampingOverride=0.999
    bGravityOverrideInSimSpace=true
    bOverrideAngularDamping=true
    bUseGravityOverride=true
    alphaBoolBlend {blendOption=Linear}
    boundBone {boneName=dyn_bootloop_l}
    componentAppliedLinearAccClamp=<3000 3000 3000>
    componentLinearAccScale=<0.3 0.3 0.3>
    componentLinearVelScale=<0.3 0.3 0.3>
    gravityOverride {z=-980}
    physicsBodyDefinitions [{
      sphereCollisionRadius=0
      boundBone {boneName=dyn_bootloop_l}
      boxExtents=<10 12 12>
      constraintSetup {angularLimitsMax=<5 0 30>, angularLimitsMin=<0 0 0>, angularTarget=<0 0 0>}
      localJointOffset=<2 0.2 0.2>
     }]
   }
   n11 = Anim Dynamics - Bone: dyn_boottongue_r(ComponentPose=@n10, Alpha=1, ExternalForce="0,0,0") {
    angularDampingOverride=0.999
    angularSpringConstant=200
    bAngularSpring=true
    bGravityOverrideInSimSpace=true
    bUseGravityOverride=true
    simulationSpace=BoneRelative
    alphaBoolBlend {blendOption=Linear}
    boundBone {boneName=dyn_boottongue_r}
    componentAppliedLinearAccClamp=<3000 3000 3000>
    componentLinearAccScale=<0.3 0.3 0.3>
    componentLinearVelScale=<0.3 0.3 0.3>
    gravityOverride {z=-980}
    physicsBodyDefinitions [{
      sphereCollisionRadius=0
      boundBone {boneName=dyn_boottongue_r}
      boxExtents=<6 7 7>
      constraintSetup {twistAxis=AxisY, angularLimitsMax=<5 0 10>, angularLimitsMin=<0 0 -20>, angularTarget=<5 0 0>}
      localJointOffset=<4 0.2 0.2>
     }]
    relativeSpaceBone {boneName=calf_twist_01_r}
   }
   n12 = Anim Dynamics - Bone: dyn_bootloop_r(ComponentPose=@n11, Alpha=1, ExternalForce="0,0,0") {
    angularDampingOverride=0.999
    bGravityOverrideInSimSpace=true
    bOverrideAngularDamping=true
    bUseGravityOverride=true
    alphaBoolBlend {blendOption=Linear}
    boundBone {boneName=dyn_bootloop_r}
    componentAppliedLinearAccClamp=<3000 3000 3000>
    componentLinearAccScale=<0.3 0.3 0.3>
    componentLinearVelScale=<0.3 0.3 0.3>
    gravityOverride {z=-980}
    physicsBodyDefinitions [{
      sphereCollisionRadius=0
      boundBone {boneName=dyn_bootloop_r}
      boxExtents=<10 12 12>
      constraintSetup {angularLimitsMax=<5 0 30>, angularLimitsMin=<0 0 0>, angularTarget=<0 0 0>}
      localJointOffset=<2 0.2 0.2>
     }]
   }
   Save cached pose 'bootDyn'(Pose=Component To Local(ComponentPose=@n12)) {cacheName=bootDyn}
   n13 = Control Rig!(
    Source=Copy Pose From Mesh() {bUseAttachedParent=true}
    Alpha
    targeting_alpha:double=Get IsTargeting() -> (IsTargeting:double)
    Running_alpha:double=Get isRunning() -> (isRunning:double)
   ) -> (Pose) {
    alphaBoolBlend {blendOption=Linear}
    alphaScaleBiasClamp {bInterpResult=true, interpSpeedDecreasing=5, interpSpeedIncreasing=5}
    controlRigAssetReference {blueprintRigClass=&/F_MED_RebelClaw_Aviator_CtrlRig.F_MED_RebelClaw_Aviator_CtrlRig_C}
   }
   n14 = Control Rig!(Source=@n13, Alpha) -> (Pose) {
    alphaBoolBlend {blendOption=Linear}
    controlRigAssetReference {blueprintRigClass=&/F_MED_DeformRig_RebelClaw_Aviator_ControlRig.F_MED_DeformRig_RebelClaw_Aviator_ControlRig_C}
   }
   n15 = Get UpperArmRot_r() -> (UpperArmRot_r:double)
   n16 = "Constraint - shirtmover_r (PP)"!(ComponentPose=Local To Component(LocalPose=@n14), Alpha=@n15, ConstraintWeights_0=0.6, ConstraintWeights_1:float=0.4) -> (Pose) {
    alphaBoolBlend {blendOption=Linear}
    alphaScaleBiasClamp {bInterpResult=true, interpSpeedDecreasing=7}
    boneToModify {boneName=shirtmover_r}
    constraintSetup [{targetBone {boneName=clavicle_r}}, {targetBone {boneName=upperarm_r}}]
    constraintWeights [1,1]
   }
   n17 = "Constraint - shirtmover_l (PP)"!(ComponentPose=@n16, Alpha=Get UpperArmRot_l() -> (UpperArmRot_l:double), ConstraintWeights_0=0.6, ConstraintWeights_1:float=0.4) -> (Pose) {
    alphaBoolBlend {blendOption=Linear}
    alphaScaleBiasClamp {bInterpResult=true, interpSpeedDecreasing=7}
    boneToModify {boneName=shirtmover_l}
    constraintSetup [{targetBone {boneName=clavicle_l}}, {targetBone {boneName=upperarm_l}}]
    constraintWeights [1,1]
   }
   n18 = "Transform (Modify) Bone - Bone: shirtmover_r"(ComponentPose=@n17, Alpha=@n15, Translation="0,0,4.7", Rotation="0,0,0", Scale="1,1,1") {
    rotationMode=BMM_Additive
    rotationSpace=BCS_BoneSpace
    translationMode=BMM_Additive
    translationSpace=BCS_BoneSpace
    alphaBoolBlend {blendOption=Linear}
    alphaScaleBiasClamp {bInterpResult=true, interpSpeedDecreasing=2, interpSpeedIncreasing=2}
    boneToModify {boneName=shirtmover_r}
    translation {z=4.7}
   }
   n19 = RigidBody!(ComponentPose=@n18, Alpha, OverrideWorldGravity:Vector=Get GravityOverride(), ExternalForce) -> (Pose) {
    bClampLinearTranslationLimitToRefPose=true
    bOverrideWorldGravity=true
    overridePhysicsAsset=&/F_MED_RebelClaw_Aviator_Physics
    simulationSpace=BaseBoneSpace
    alphaBoolBlend {blendOption=Linear}
    baseBoneRef {boneName=spine_04}
    componentAppliedLinearAccClamp=<2000 2000 1000>
    componentLinearAccScale {x=0.2}
    componentLinearVelScale {x=0.2}
    overrideWorldGravity {z=-980}
    simSpaceSettings {maxAngularAcceleration=100, maxAngularVelocity=5, maxLinearAcceleration=300, maxLinearVelocity=50, velocityScaleZ=0, worldAlpha=0.5}
   }
   Save cached pose 'startpose'(Pose=Component To Local(ComponentPose=@n19)) {cacheName=startpose}
  ]}
 EventGraph {nodes [
   Event BlueprintUpdateAnimation()
   > n1 = Sequence!(execute) -> (then_0, then_1, then_2:Exec)
   > n2 = Gravity Override State Function DEPRECATED(CharacterPartAnimInstance=Self-Reference(), GameTimeInSeconds=GetGameTimeInSeconds(), GravityOverrideParams=Get Gravity Override Params())
   > Set GravityOverride(GravityOverride=@n2)
   n3 = Get bIsTargeting()
   n4 = Get bIsSkydiveFloating()
   > Set IsTargeting(IsTargeting:double="To Float (Boolean)"(InBool=OR Boolean(A=@n4, B=@n3))) -> (Output_Get:double)
   n5 = AND Boolean!(A=NOT Boolean(A=Get bIsSurfaceSwimming()), B=Get bIsMoving2D(), C:bool=NOT Boolean(A=Get bIsSkyDiving())) -> (ReturnValue)
   n6 = Get bIsSprinting()
   > Set isRunning(isRunning:double="To Float (Boolean)"(InBool=OR Boolean(A=@n6, B=@n5))) -> (Output_Get:double)
   n7 = Get bIsParachuteOpened()
   n8 = Get bIsSkydiveFloating()
   > Set Strings out(Strings out:double="To Float (Boolean)"(InBool=OR Boolean(A=@n7, B=@n8))) -> (Output_Get:double)
   n9 = Get Gun_out() -> (Gun_out:bool)
   n10 = Get bIsParachuteOpened()
   n11 = Get bIsSkydiveFloating()
   > Set String_r_out(String_r_out:double="To Float (Boolean)"(InBool=OR Boolean!(A=@n10, B=@n11, C:bool=@n9) -> (ReturnValue))) -> (Output_Get:double)
   n12 = Get bIsSkydiveFloating()
   n13 = Get bIsParachuteOpened()
   > Set isSkydiving(isSkydiving:double="To Float (Boolean)"(InBool=OR Boolean!(A=@n13, B=@n12) -> (ReturnValue))) -> (Output_Get:double)
   n14 = Get bIsSkydiveDiveMode()
   n15 = Get bIsSkydiveFloating()
   > Set SkydivingFloating_PointingDown(SkydivingFloating_PointingDown:double="To Float (Boolean)"(InBool=OR Boolean(A=@n15, B=@n14))) -> (Output_Get:double)
   n16 = Sequence!(execute=@n1.then_1) -> (then_0, then_1)
   n17 = GetOwnerSkeletalMeshComponent()
   n18 = BreakRotator(InRot=BreakTransform(InTransform=GetDeltaTransformFromRefPose(self=@n17, BoneName=upperarm_r, BaseName=clavicle_r)).Rotation)
   > n19 = Branch(Condition=OR Boolean(A="float <= float"(A=@n18.Pitch, B=-50) -> (ReturnValue=~), B="float <= float"(A=@n18.Yaw, B=3) -> (ReturnValue=~)))
   > Set UpperArmRot_r(UpperArmRot_r:double=1) -> (Output_Get:double)
   Set UpperArmRot_r(execute=@n19.else, UpperArmRot_r=0)
   n20 = GetOwnerSkeletalMeshComponent()
   n21 = BreakRotator(InRot=BreakTransform(InTransform=GetDeltaTransformFromRefPose(self=@n20, BoneName=upperarm_l, BaseName=clavicle_l)).Rotation)
   n22 = Branch(
    execute=@n16.then_1
    Condition=OR Boolean(A="float <= float"(A=@n21.Pitch, B=-50) -> (ReturnValue=~), B="float <= float"(A=@n21.Yaw, B=3) -> (ReturnValue=~))
   )
   > Set UpperArmRot_l(UpperArmRot_l:double=1) -> (Output_Get:double)
   Set UpperArmRot_l(execute=@n22.else, UpperArmRot_l=0)
   n23 = Cast To FortPlayerAnimInstance(execute=@n1.then_2, Object=GetAnimInstance(self=GetOwnerSkeletalMeshComponent()))
   > Set ar_firingPosition_GC(ar_firingPosition_GC:int=0) -> (Output_Get:int=0)
   > n24 = Switch on EFortWeaponCoreAnimation!(execute, Selection:EFortWeaponCoreAnimation Enum=Get WeaponCoreAnim(self=@n23)) -> (
    Melee:Exec
    Pistol:Exec
    Shotgun:Exec
    PaperBlueprint:Exec
    Rifle:Exec
    MeleeOneHand:Exec
    MachinePistol:Exec
    RocketLauncher:Exec
    GrenadeLauncher:Exec
    GoingCommando:Exec
    AssaultRifle:Exec
    TacticalShotgun:Exec
    SniperRifle:Exec
    TrapPlacement:Exec
    ShoulderLauncher:Exec
    AbilityDecoTool:Exec
    Crossbow:Exec
    C4:Exec
    RemoteControl:Exec
    DualWield:Exec
    AR_BullPup:Exec
    AR_ForwardGrip:Exec
    MedPackPaddles:Exec
    SMG_P90:Exec
    AR_DrumGun:Exec
    Consumable_Small:Exec
    Consumable_Large:Exec
    Balloon:Exec
    MountedTurret:Exec
    CreativeTool:Exec
    ExplosiveBow:Exec
    AshtonIndigo:Exec
    AshtonChicago:Exec
    MeleeDualWield:Exec
    Unarmed:Exec
   )
   n25 = Branch(execute=[@n24.GoingCommando, @n24.MachinePistol, @n24.Pistol, @n24.Rifle, @n24.RocketLauncher, @n24.Shotgun, @n24.SniperRifle, @n24.TacticalShotgun])
   > n26 = Set ar_firingPosition_GC(ar_firingPosition_GC=1) -> (Output_Get=0)
   > Set Gun_out(Gun_out:bool="To Boolean (Integer)"!(InInt:int=@n26) -> (ReturnValue:bool)) -> (Output_Get:bool)
   Set ar_firingPosition_GC(execute=@n25.else, ar_firingPosition_GC=0) -> (Output_Get=0)
   Event OnCharacterCustomizationLoaded()
   > Set BodyMesh(BodyMesh=GetOwnerSkeletalMeshComponent())
   > n27 = Cast To PlayerPawn_Athena(Object=TryGetPawnOwner())
   > n28 = Set PawnOwner(PawnOwner:Player Pawn Athena Object Reference=@n27) -> (Output_Get:Player Pawn Athena Object Reference)
   > Set TailMesh(TailMesh:Skeletal Mesh Component Object Reference=GetSkeletalMeshForPartType(self=@n28, PartType=MiscOrTail)) -> (Output_Get:Skeletal Mesh Component Object Reference)
   n29 = CheckForBackbling()
   > Set isWearingBackbling(isWearingBackbling=@n29)
   > Set isWearingCape(isWearingCape=@n29.bIsBackblingLong)
   > Set characterHasTail(characterHasTail:bool=CheckForTail!(self:Custom Character Part Anim Instance Object Reference) -> (bHasTail:bool)) -> (Output_Get:bool)
   > Branch(Condition=Get characterHasTail() -> (characterHasTail:bool))
   > n30 = Branch(Condition=Get isWearingCape() -> (isWearingCape:bool))
   n31 = Get TailMesh() -> (TailMesh:Skeletal Mesh Component Object Reference)
   > SetVisibility!(execute, self:Scene Component Object Reference=@n31, bNewVisibility:bool=false, bPropagateToChildren:bool=false) -> (then)
   SetVisibility(execute=@n30.else, self=@n31, bNewVisibility=true, bPropagateToChildren=false)
  ]}
]
variables [
 ar_firingPosition_GC {type=integer}
 Body_Mesh {type=SkeletalMeshComponent}
 BodyMesh {type=SkeletalMeshComponent}
 characterHasTail {type=boolean}
 Gravity Override Params {type=GravityOverrideParamsStruct}
 GravityOverride {type=Vector}
 Gun_out {type=boolean}
 isRunning {type=number}
 isSkydiving {type=number}
 IsTargeting {type=number}
 isWearingBackbling {type=boolean}
 isWearingCape {type=boolean}
 PawnOwner {type=PlayerPawn_Athena_C}
 SkydivingFloating_PointingDown {type=number}
 String_r_out {type=number}
 Strings out {type=number}
 TailMesh {type=SkeletalMeshComponent}
 UpperArmRot_l {type=number}
 UpperArmRot_r {type=number}
]
