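# Physics asset for the F_MED_RebelClaw_Aviator face accessory, giving capsule bodies to its bangs, cap flaps and pigtails.
# Bangs and cap flaps chain from the head through base bones, several of them locked, into narrowly limited segments.
# Each pigtail fans limited segments from a root locked to its base, which joins spine_05 with free rotation.
# Larger head, upper-chest, upper-arm and hood capsules add body volumes, linked to spine_05 by free-rotation constraints.
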PhysicsAsset /BRCosmetics/Characters/Player/Female/Medium/Bodies/F_MED_RebelClaw_Aviator/Meshes/Parts/F_MED_RebelClaw_Aviator_FaceAcc_Physics
bodies {
 bangs_l [{shapeType=Capsule, length=1, radius=1, shapeName=bangs_l_capsule, center=<2 0 0>, rotation=<rot 90 0 0>}]
 bangs_mid [{shapeType=Capsule, length=1, radius=1, shapeName=bangs_mid_capsule, center=<2 0 0>, rotation=<rot 90 0 0>}]
 base_bangs_inner_r [{shapeType=Capsule, length=0.5, radius=0.5, shapeName=base_bangs_inner_r_capsule, center=<1 0 0>}]
 base_bangs_l [{shapeType=Capsule, length=0.5, radius=0.5, shapeName=base_bangs_l_capsule}]
 base_bangs_mid [{shapeType=Capsule, length=0.5, radius=0.5, shapeName=base_bangs_mid_capsule}]
 base_bangs_outer_r [{shapeType=Capsule, length=0.5, radius=0.5, shapeName=base_bangs_outer_r_capsule}]
 base_capflap_l [{shapeType=Capsule, length=0.5, radius=0.5, shapeName=base_capflap_l_capsule}]
 base_capflap_r [{shapeType=Capsule, length=0.5, radius=0.5, shapeName=base_capflap_r_capsule}]
 base_pigtail_l [{shapeType=Capsule, length=0.505, radius=0.505, shapeName=base_pigtail_l_capsule, rotation=<rot 0 -90 0>}]
 base_pigtail_r [{shapeType=Capsule, length=0.505, radius=0.505, shapeName=base_pigtail_r_capsule, rotation=<rot 0 -90 0>}]
 capflap_lower_l [{shapeType=Capsule, length=1, radius=3, shapeName=capflap_lower_l_capsule, center=<5 0 0>, rotation=<rot 90 0 0>}]
 dyn_bangs_inner_r [{shapeType=Capsule, length=1, radius=1, shapeName=dyn_bangs_inner_r_capsule, center=<2 0 0>, rotation=<rot 90 0 0>}]
 dyn_bangs_l [{shapeType=Capsule, length=1, radius=1, shapeName=dyn_bangs_l_capsule, center=<2 0 0>, rotation=<rot 90 0 0>}]
 dyn_bangs_mid [{shapeType=Capsule, length=1, radius=1, shapeName=dyn_bangs_mid_capsule, center=<2 0 0>, rotation=<rot 90 0 0>}]
 dyn_bangs_outer_r [{shapeType=Capsule, length=1, radius=0.5, shapeName=dyn_bangs_outer_r_capsule, center=<1 0 0>, rotation=<rot 90 0 0>}]
 dyn_bangs_outer_r1 [{shapeType=Capsule, length=1, radius=0.5, shapeName=dyn_bangs_outer_r1_capsule, center=<1.5 0 0>, rotation=<rot 90 0 0>}]
 dyn_capflap_l [{shapeType=Capsule, length=2, radius=3, shapeName=dyn_capflap_l_capsule, center=<4.2762 -0.283 0>, rotation=<rot 90 180 171.5613>}]
 dyn_capflap_lower_l [{shapeType=Capsule, length=2, radius=2.7, shapeName=dyn_capflap_lower_l_capsule, center=<6.044 -0.1171 1.7907>, rotation=<rot 90 0 0>}]
 dyn_capflap_lower_r [{
   shapeType=Capsule
   length=2
   radius=2.7
   shapeName=dyn_capflap_lower_r_capsule
   center=<5.1281 0.1565 -2.8054>
   rotation=<rot 74.3652 0.0003 0.0001>
  }]
 dyn_capflap_r [{shapeType=Capsule, length=2.5, radius=3, shapeName=dyn_capflap_r_capsule, center=<3.556 -0.0424 0>, rotation=<rot 66.8072 8.3169 0>}]
 dyn_pigtail_10_r [{shapeType=Capsule, length=1, radius=1.5, shapeName=dyn_pigtail_10_r_capsule, center=<3.5 0 0>, rotation=<rot 90 0 0>}]
 dyn_pigtail_11_r [{shapeType=Capsule, length=2, radius=2, shapeName=dyn_pigtail_11_r_capsule, center=<3 0.4 -1.6>, rotation=<rot 0 0 90>}]
 dyn_pigtail_14_l [{shapeType=Capsule, length=1, radius=1.3, shapeName=dyn_pigtail_14_l_capsule, center=<3.5 0 0>, rotation=<rot 90 0 0>}]
 dyn_pigtail_15_l [{shapeType=Capsule, length=1, radius=1.3, shapeName=dyn_pigtail_15_l_capsule, center=<3 0 0>, rotation=<rot 90 0 0>}]
 dyn_pigtail_16_l [{shapeType=Capsule, length=1, radius=1.5, shapeName=dyn_pigtail_16_l_capsule, center=<2 0 0>, rotation=<rot 90 0 0>}]
 dyn_pigtail_6_l [{shapeType=Capsule, length=1, radius=1, shapeName=dyn_pigtail_6_l_capsule, center=<3 0 0>, rotation=<rot 90 0 0>}]
 dyn_pigtail_6_r [{shapeType=Capsule, length=1, radius=1, shapeName=dyn_pigtail_6_r_capsule, center=<3 0 0>, rotation=<rot 90 0 0>}]
 dyn_pigtail_8_l [{shapeType=Capsule, length=1, radius=1, shapeName=dyn_pigtail_8_l_capsule, center=<2.3 0 0>, rotation=<rot 90 0 0>}]
 dyn_pigtail_8_r [{shapeType=Capsule, length=1, radius=1, shapeName=dyn_pigtail_8_r_capsule, center=<2 0 0>, rotation=<rot 90 0 0>}]
 dyn_pigtail_9_l [{shapeType=Capsule, length=1, radius=1, shapeName=dyn_pigtail_9_l_capsule, center=<2 0 0>, rotation=<rot 90 0 0>}]
 dyn_pigtail_9_r [{shapeType=Capsule, length=1, radius=1, shapeName=dyn_pigtail_9_r_capsule, center=<3.5 0 0>, rotation=<rot 90 0 0>}]
 dyn_pigtail_l [{shapeType=Capsule, length=1, radius=1, shapeName=dyn_pigtail_l_capsule, rotation=<rot 90 0 0>}]
 dyn_pigtail_r [{shapeType=Capsule, length=1, radius=1, shapeName=dyn_pigtail_r_capsule, rotation=<rot 90 0 0>}]
 head [{shapeType=Capsule, length=6.1875, radius=9.5625, shapeName=head_capsule, center=<0 -0.1914 0>, rotation=<rot 90 0 0>}]
 hoodmover_b [{shapeType=Capsule, length=12, radius=7, shapeName=hoodmover_b_capsule, center=<-0.7 4.7 0>, rotation=<rot 0 -90 0>}]
 spine_05 [{shapeType=Capsule, length=8, radius=13, shapeName=spine_05_capsule, center=<3.1708 -3.5 0>}]
 upperarm_twist_01_l [{shapeType=Capsule, length=20, radius=4.5, shapeName=upperarm_twist_01_l_capsule, center=<4 0 0>, rotation=<rot 90 0 0>}]
 upperarm_twist_01_r [{shapeType=Capsule, length=20, radius=4.5, shapeName=upperarm_twist_01_r_capsule, center=<-4 0 0>, rotation=<rot 90 0 0>}]
}
constraints [
 {
  bone1Name=base_bangs_inner_r
  bone2Name=head
  swing1LimitDegrees=20
  swing1Motion=Locked
  swing2LimitDegrees=20
  swing2Motion=Locked
  twistLimitDegrees=1
  twistMotion=Locked
 }
 {bone1Name=dyn_bangs_inner_r, bone2Name=base_bangs_inner_r, swing1LimitDegrees=10, swing2LimitDegrees=10, twistLimitDegrees=1}
 {
  bone1Name=base_bangs_l
  bone2Name=head
  swing1LimitDegrees=20
  swing1Motion=Locked
  swing2LimitDegrees=20
  swing2Motion=Locked
  twistLimitDegrees=1
  twistMotion=Locked
 }
 {bone1Name=bangs_l, bone2Name=base_bangs_l, swing1LimitDegrees=20, swing2LimitDegrees=20, twistLimitDegrees=1}
 {bone1Name=base_bangs_mid, bone2Name=head, swing1LimitDegrees=20, swing2LimitDegrees=20, twistLimitDegrees=1}
 {bone1Name=bangs_mid, bone2Name=base_bangs_mid, swing1LimitDegrees=20, swing2LimitDegrees=20, twistLimitDegrees=1}
 {bone1Name=base_bangs_outer_r, bone2Name=head, swing1LimitDegrees=20, swing2LimitDegrees=20, twistLimitDegrees=1}
 {bone1Name=dyn_bangs_outer_r, bone2Name=base_bangs_outer_r, swing1LimitDegrees=2, swing2LimitDegrees=2, twistLimitDegrees=0}
 {bone1Name=dyn_bangs_outer_r1, bone2Name=dyn_bangs_outer_r, swing1LimitDegrees=5, swing2LimitDegrees=5, twistLimitDegrees=0}
 {
  bone1Name=base_capflap_l
  bone2Name=head
  swing1LimitDegrees=20
  swing1Motion=Locked
  swing2LimitDegrees=20
  swing2Motion=Locked
  twistLimitDegrees=1
  twistMotion=Locked
 }
 {bone1Name=dyn_capflap_l, bone2Name=base_capflap_l, swing1LimitDegrees=10, swing2LimitDegrees=10, twistLimitDegrees=1}
 {bone1Name=capflap_lower_l, bone2Name=dyn_capflap_l, swing1LimitDegrees=20, swing2LimitDegrees=20, twistLimitDegrees=1}
 {
  bone1Name=base_capflap_r
  bone2Name=head
  swing1LimitDegrees=20
  swing1Motion=Locked
  swing2LimitDegrees=20
  swing2Motion=Locked
  twistLimitDegrees=1
  twistMotion=Locked
 }
 {bone1Name=dyn_capflap_r, bone2Name=base_capflap_r, swing1LimitDegrees=10, swing2LimitDegrees=10, twistLimitDegrees=1}
 {bone1Name=dyn_capflap_lower_r, bone2Name=dyn_capflap_r, swing1LimitDegrees=10, swing2LimitDegrees=20, twistLimitDegrees=1}
 {bone1Name=dyn_capflap_lower_l, bone2Name=dyn_capflap_l, swing1LimitDegrees=10, swing2LimitDegrees=20, twistLimitDegrees=1}
 {bone1Name=dyn_bangs_l, bone2Name=base_bangs_l, swing1LimitDegrees=10, swing2LimitDegrees=10, twistLimitDegrees=0}
 {bone1Name=dyn_bangs_mid, bone2Name=base_bangs_mid, swing1LimitDegrees=10, swing2LimitDegrees=10, twistLimitDegrees=0}
 {
  bone1Name=dyn_pigtail_l
  bone2Name=base_pigtail_l
  swing1LimitDegrees=5
  swing1Motion=Locked
  swing2LimitDegrees=5
  swing2Motion=Locked
  twistLimitDegrees=0
  twistMotion=Locked
 }
 {
  bone1Name=dyn_pigtail_r
  bone2Name=base_pigtail_r
  swing1LimitDegrees=5
  swing1Motion=Locked
  swing2LimitDegrees=5
  swing2Motion=Locked
  twistLimitDegrees=0
  twistMotion=Locked
 }
 {
  bone1Name=head
  bone2Name=spine_05
  swing1LimitDegrees=45
  swing1Motion=Free
  swing2LimitDegrees=45
  swing2Motion=Free
  twistLimitDegrees=45
  twistMotion=Free
 }
 {
  bone1Name=base_pigtail_r
  bone2Name=spine_05
  swing1LimitDegrees=45
  swing1Motion=Free
  swing2LimitDegrees=45
  swing2Motion=Free
  twistLimitDegrees=45
  twistMotion=Free
 }
 {
  bone1Name=base_pigtail_l
  bone2Name=spine_05
  swing1LimitDegrees=45
  swing1Motion=Free
  swing2LimitDegrees=45
  swing2Motion=Free
  twistLimitDegrees=45
  twistMotion=Free
 }
 {
  bone1Name=upperarm_twist_01_l
  bone2Name=spine_05
  swing1LimitDegrees=45
  swing1Motion=Free
  swing2LimitDegrees=45
  swing2Motion=Free
  twistLimitDegrees=45
  twistMotion=Free
 }
 {
  bone1Name=upperarm_twist_01_r
  bone2Name=spine_05
  swing1LimitDegrees=45
  swing1Motion=Free
  swing2LimitDegrees=45
  swing2Motion=Free
  twistLimitDegrees=45
  twistMotion=Free
 }
 {
  bone1Name=hoodmover_b
  bone2Name=spine_05
  swing1LimitDegrees=45
  swing1Motion=Free
  swing2LimitDegrees=45
  swing2Motion=Free
  twistLimitDegrees=45
  twistMotion=Free
 }
 {bone1Name=dyn_pigtail_8_l, bone2Name=dyn_pigtail_l, swing1LimitDegrees=10, swing2LimitDegrees=10, twistLimitDegrees=0}
 {bone1Name=dyn_pigtail_9_l, bone2Name=dyn_pigtail_l, swing1LimitDegrees=10, swing2LimitDegrees=10, twistLimitDegrees=0}
 {bone1Name=dyn_pigtail_6_l, bone2Name=dyn_pigtail_l, swing1LimitDegrees=10, swing2LimitDegrees=10, twistLimitDegrees=0}
 {bone1Name=dyn_pigtail_16_l, bone2Name=dyn_pigtail_l, swing1LimitDegrees=10, swing2LimitDegrees=10, twistLimitDegrees=0}
 {bone1Name=dyn_pigtail_15_l, bone2Name=dyn_pigtail_l, swing1LimitDegrees=10, swing2LimitDegrees=10, twistLimitDegrees=0}
 {bone1Name=dyn_pigtail_14_l, bone2Name=dyn_pigtail_l, swing1LimitDegrees=10, swing2LimitDegrees=10, twistLimitDegrees=0}
 {bone1Name=dyn_pigtail_10_r, bone2Name=dyn_pigtail_r, swing1LimitDegrees=10, swing2LimitDegrees=10, twistLimitDegrees=0}
 {bone1Name=dyn_pigtail_6_r, bone2Name=dyn_pigtail_r, swing1LimitDegrees=10, swing2LimitDegrees=10, twistLimitDegrees=0}
 {bone1Name=dyn_pigtail_9_r, bone2Name=dyn_pigtail_r, swing1LimitDegrees=10, swing2LimitDegrees=10, twistLimitDegrees=0}
 {bone1Name=dyn_pigtail_8_r, bone2Name=dyn_pigtail_r, swing1LimitDegrees=10, swing2LimitDegrees=10, twistLimitDegrees=0}
 {bone1Name=dyn_pigtail_11_r, bone2Name=dyn_pigtail_r, swing1LimitDegrees=5, swing2LimitDegrees=5, twistLimitDegrees=0}
]
