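# Control Rig for the masked face accessory on the female RebelClaw Aviator, applying corrective offsets to bones including the pigtail bases.
# A spherical pose reader tracks head orientation against spine_04 and stores neck_bend_alpha, with debug drawing left on.
# That alpha weights a first transform offset.
# A second offset is weighted by a blend of Targeting_alpha and an unclamped remap of the neck alpha, biased toward the latter.
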ControlRigBlueprint /BRCosmetics/Characters/Player/Female/Medium/Bodies/F_MED_RebelClaw_Aviator/Meshes/Parts/F_MED_RebelClaw_Aviator_Masked_FaceAcc_CtrlRig
graphs [
 RigVMFunctionLibrary {links [], nodes []}
 RigVMModel {nodes [
   RigUnit_BeginExecution = &RigUnit_BeginExecution()
   n1 = &RigVMFunction_Sequence(ExecuteContext=@RigUnit_BeginExecution.ExecutePin)
   n2 = &RigUnit_SphericalPoseReader(
    ActiveRegionScaleFactors=(PositiveWidth=3,NegativeWidth=0.3,PositiveHeight=6,NegativeHeight=6)
    ActiveRegionSize=0.1
    Debug=(bDrawDebug=True,DebugScale=25,DebugSegments=20,DebugThickness=0.25)
    DriverAxis=(X=1,Y=0,Z=0)
    DriverItem=(Type=Bone,Name="head")
    FalloffRegionScaleFactors=(PositiveWidth=-3,NegativeWidth=1,PositiveHeight=0,NegativeHeight=0)
    FalloffSize=0.25
    OptionalParentItem=(Type=Bone,Name="spine_04")
    RotationOffset=(X=90,Y=0,Z=-140)
    ExecutePin=@n1.A
   )
   > RigVMVariableNode(Value=@n2.OutputParam) {variable=neck_bend_alpha}
   &RigUnit_ModifyTransforms(
    ItemToModify="((Item=(Type=Bone,Name=\"base_pigtail_l\"),Transform=(Rotation=(X=0,Y=0,Z=-0.4226,W=0.9063),Translation=(X=0,Y=0,Z=0),Scale3D=(X=1,Y=1,Z=1))),(Item=(Type=Bone,Name=\"base_pigtail_r\"),Transform=(Rotation=(X=0,Y=0,Z=-0.4226,W=0.9063),Translation=(X=0,Y=0,Z=0),S ...(truncated)"
    ExecutePin=@n1.B
    Weight=RigVMVariableNode() {variable=neck_bend_alpha}.Value
   )
   n3 = RigVMVariableNode() {variable=neck_bend_alpha}
   > &RigUnit_ModifyTransforms(
    ItemToModify="((Item=(Type=Bone,Name=\"base_pigtail_l\"),Transform=(Rotation=(X=-0.13,Y=0.0114,Z=-0.0864,W=0.9877),Translation=(X=0,Y=0,Z=0),Scale3D=(X=1,Y=1,Z=1))),(Item=(Type=Bone,Name=\"base_pigtail_r\"),Transform=(Rotation=(X=-0.1285,Y=0.0227,Z=-0.1722,W=0.9764),Transla ...(truncated)"
    Weight=/Script/RigVM.RigVMFunction_MathFloatLerp(
     T=0.7
     B=&RigVMFunction_MathDoubleRemap(SourceMaximum=0.25, SourceMinimum=0, TargetMaximum=1, TargetMinimum=0, bClamp=False, Value=@n3.Value).Result
     A=RigVMVariableNode() {variable=Targeting_alpha}.Value
    ) {is=struct RigVMUnitNode, template="Interpolate::Execute(in A,in B,in T,out Result)"}.Result
   )
  ]}
]
hierarchy [
 root {loc=<0 0 0>, rot=<0 0 0 1>}
 pelvis {parent=root, loc=<0 -0.5301 98.0571>, rot=<0.0127 0.707 -0.0127 -0.707>}
 spine_01 {parent=pelvis, loc=<6.1208 -0.9328 0>, rot=<0 0 0.034 -0.9994>}
 spine_02 {parent=spine_01, loc=<6.5996 -0.2095 0>, rot=<0 0 0.0238 -0.9997>}
 spine_03 {parent=spine_02, loc=<6.4914 -0.1237 0>, rot=<0 0 0.0149 -0.9999>}
 spine_04 {parent=spine_03, loc=<6.9598 0.0794 0>, rot=<0 0 -0.0132 -0.9999>}
 spine_05 {parent=spine_04, loc=<6.9987 0 0>, rot=<0 0 -0.0015 -1>}
 neck_01 {parent=spine_05, loc=<11.3138 0.1529 0>, rot=<0 0 -0.2003 -0.9797>}
 neck_02 {parent=neck_01, loc=<4.5627 0 0>, rot=<0 0 0.0504 -0.9987>}
 head {parent=neck_02, loc=<5.3914 0 0>, rot=<0 0 0.115 -0.9934>}
 base_capflap_r {parent=head, loc=<1.0636 1.9294 6.3707>, rot=<-0.1735 -0.9839 0.043 0.0076>}
 dyn_capflap_r {parent=base_capflap_r, loc=<0 0 0>, rot=<0 0 0 1>}
 dyn_capflap_lower_r {parent=dyn_capflap_r, loc=<5.6219 0.7819 -0.5525>, rot=<0.0472 -0.0889 -0.0396 -0.9941>}
 base_capflap_l {parent=head, loc=<1.0636 1.9294 -6.6266>, rot=<-0.2164 -0.9763 0.0003 -0.0015>}
 dyn_capflap_l {parent=base_capflap_l, loc=<0 0 0>, rot=<0 0 0 1>}
 dyn_capflap_lower_l {parent=dyn_capflap_l, loc=<5.5905 0.3339 0.3644>, rot=<0.0275 0.1668 -0.0525 -0.9842>}
 base_pigtail_r {parent=head, loc=<-0.0264 -5.581 5.2>, rot=<0.3894 -0.9006 0.156 0.1138>}
 dyn_pigtail_r {parent=base_pigtail_r, loc=<0 0 0>, rot=<0 0 0 1>}
 dyn_pigtail_1_r {parent=dyn_pigtail_r, loc=<1.288 -2.3765 1.7135>, rot=<0 0 0.5 -0.866>}
 dyn_pigtail_2_r {parent=dyn_pigtail_r, loc=<2.794 -1.7792 2.8054>, rot=<0.0594 0.1632 0.3368 -0.9254>}
 dyn_pigtail_3_r {parent=dyn_pigtail_r, loc=<3.7888 -1.9375 -0.6064>, rot=<-0.171 -0.2962 0.3214 -0.883>}
 dyn_pigtail_4_r {parent=dyn_pigtail_r, loc=<1.4159 -1.9468 -1.4416>, rot=<-0.1962 -0.2802 0.539 -0.7698>}
 dyn_pigtail_5_r {parent=dyn_pigtail_r, loc=<0.9544 0.5515 -2.576>, rot=<-0.1255 -0.6333 0.2961 -0.7039>}
 dyn_pigtail_6_r {parent=dyn_pigtail_r, loc=<2.7401 2.9148 -2.0479>, rot=<-0.3464 -0.4321 -0.1229 -0.8235>}
 dyn_pigtail_7_r {parent=dyn_pigtail_r, loc=<4.4671 -0.1614 -2.2428>, rot=<0 -0.342 0 -0.9397>}
 dyn_pigtail_8_r {parent=dyn_pigtail_r, loc=<4.943 2.2515 -0.9079>, rot=<-0.0392 -0.2981 -0.1245 -0.9456>}
 dyn_pigtail_9_r {parent=dyn_pigtail_r, loc=<5.4087 0.6357 -0.0821>, rot=<0 0.0436 0 -0.999>}
 dyn_pigtail_10_r {parent=dyn_pigtail_r, loc=<3.921 0.6131 2.5405>, rot=<0 0.2588 0 -0.9659>}
 dyn_pigtail_11_r {parent=dyn_pigtail_r, loc=<2.2777 -0.3233 3.8216>, rot=<0 0.5 0 -0.866>}
 dyn_pigtail_12_r {parent=dyn_pigtail_r, loc=<0.5378 2.848 -0.733>, rot=<-0.3304 -0.6009 0.0192 -0.7276>}
 dyn_pigtail_13_r {parent=dyn_pigtail_r, loc=<3.3619 -1.6885 -3.0462>, rot=<-0.171 -0.2962 0.3214 -0.883>}
 base_pigtail_l {parent=head, loc=<-0.0264 -5.581 -5.2>, rot=<0.3466 -0.93 -0.1113 -0.0519>}
 dyn_pigtail_l {parent=base_pigtail_l, loc=<0.0152 -0.0144 0.0832>, rot=<-0.0698 0.0256 0.9957 0.0546>}
 dyn_pigtail_1_l {parent=dyn_pigtail_l, loc=<-1.3943 2.4397 -1.411>, rot=<0 0 0.866 0.5>}
 dyn_pigtail_2_l {parent=dyn_pigtail_l, loc=<-2.8562 2.1681 -2.0046>, rot=<-0.1632 0.0594 0.9254 0.3368>}
 dyn_pigtail_3_l {parent=dyn_pigtail_l, loc=<-3.2151 2.2231 1.2447>, rot=<0.3214 -0.117 0.8138 0.4698>}
 dyn_pigtail_4_l {parent=dyn_pigtail_l, loc=<-1.4385 2.1846 1.9037>, rot=<-0.2802 0.1962 -0.7698 -0.539>}
 dyn_pigtail_5_l {parent=dyn_pigtail_l, loc=<-0.6893 -0.3115 2.527>, rot=<0.6333 -0.1255 0.7039 0.2961>}
 dyn_pigtail_6_l {parent=dyn_pigtail_l, loc=<-2.1269 -2.4118 1.8424>, rot=<0.4321 -0.3464 0.8235 -0.1229>}
 dyn_pigtail_7_l {parent=dyn_pigtail_l, loc=<-3.9872 -0.3019 2.4157>, rot=<0.342 0 0.9397 0>}
 dyn_pigtail_8_l {parent=dyn_pigtail_l, loc=<-4.8967 -2.2482 0.8308>, rot=<0.2981 -0.0392 0.9456 -0.1245>}
 dyn_pigtail_9_l {parent=dyn_pigtail_l, loc=<-5.3804 -0.6518 0.0028>, rot=<-0.0436 0 0.999 0>}
 dyn_pigtail_10_l {parent=dyn_pigtail_l, loc=<-4.0244 0.3492 -0.6495>, rot=<0.0872 0 0.9962 0>}
 dyn_pigtail_11_l {parent=dyn_pigtail_l, loc=<-1.9571 0.3625 -4.326>, rot=<-0.3794 -0.0795 0.916 0.1036>}
 dyn_pigtail_12_l {parent=dyn_pigtail_l, loc=<-0.5095 -2.8642 0.6536>, rot=<0.6009 -0.3304 0.7276 0.0192>}
 dyn_pigtail_13_l {parent=dyn_pigtail_l, loc=<-3.2569 1.6687 2.3988>, rot=<0.2962 -0.171 0.883 0.3214>}
 dyn_pigtail_14_l {parent=dyn_pigtail_l, loc=<-3.3678 -1.0728 -4.3458>, rot=<-0.2588 0 0.9659 0>}
 dyn_pigtail_15_l {parent=dyn_pigtail_l, loc=<-4.7509 -0.0553 -3.4874>, rot=<-0.1736 0 0.9848 0>}
 dyn_pigtail_16_l {parent=dyn_pigtail_l, loc=<-4.7231 -0.5863 -1.2486>, rot=<-0.0872 0 0.9962 0>}
 dyn_pigtail_17_l {parent=dyn_pigtail_l, loc=<-0.2521 1.4381 -3.4314>, rot=<-0.4217 0.0185 0.7643 0.4875>}
 dyn_pigtail_16_l1 {parent=dyn_pigtail_l, loc=<-4.7231 -0.5863 -1.2486>, rot=<-0.0872 0 0.9962 0>}
 clavicle_l {parent=spine_05, loc=<5.4831 -0.1053 -3.2946>, rot=<0.0465 -0.7289 -0.009 -0.6829>}
 upperarm_l {parent=clavicle_l, loc=<9.8658 -1.2136 -0.4105>, rot=<0.0164 -0.3245 -0.0269 -0.9454>}
 upperarm_twist_01_l {parent=upperarm_l, loc=<7.2135 -0.1134 0>, rot=<0 0 0 1>}
 hoodmover_l {parent=clavicle_l, loc=<5.8474 -0.5147 4.9625>, rot=<0.1005 0.7235 -0.128 -0.6709>}
 clavicle_r {parent=spine_05, loc=<5.4831 -0.1053 3.2946>, rot=<0.7289 0.0465 -0.6829 0.009>}
 upperarm_r {parent=clavicle_r, loc=<-9.8658 1.2136 0.4105>, rot=<0.0164 -0.3245 -0.0269 -0.9454>}
 upperarm_twist_01_r {parent=upperarm_r, loc=<-7.2135 0.1134 0>, rot=<0 0 0 -1>}
 hoodmover_r {parent=clavicle_r, loc=<-5.8474 0.5147 -4.9625>, rot=<0.7235 -0.1005 -0.6709 0.128>}
 hoodmover_b {parent=spine_05, loc=<10.7374 -8.5713 0>, rot=<0 0 -0.1091 -0.994>}
 attach_cape {parent=spine_05, loc=<9.3216 -8.4501 0>, rot=<0.0616 -0.7044 -0.0616 -0.7044>}
]
variables [
 Neck_bend {type=double, default=0}
 Targeting_alpha {type=double, default=0}
 neck_bend_alpha {type=double, default=0}
]
